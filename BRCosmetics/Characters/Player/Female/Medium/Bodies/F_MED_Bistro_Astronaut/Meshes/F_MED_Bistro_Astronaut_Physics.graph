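# Physics asset for the female medium Bistro Astronaut cosmetic body, covering only its gun bones, dyn_gun and dyn_gun_base.
# Each bone has one capsule body, a larger offset one on the gun and a small one on the base.
# A single constraint joins the two bodies, with narrow swing limit angles and a zero twist limit angle.
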PhysicsAsset /BRCosmetics/Characters/Player/Female/Medium/Bodies/F_MED_Bistro_Astronaut/Meshes/F_MED_Bistro_Astronaut_Physics
bodies {
 dyn_gun [{shapeType=Capsule, length=9.9088, radius=2.8, shapeName=dyn_gun_capsule, center=<0 0 4.1808>, rotation=<rot 90 0 0>}]
 dyn_gun_base [{shapeType=Capsule, length=1.6233, radius=0.5969, shapeName=dyn_gun_base_capsule, rotation=<rot 90 0 0>}]
}
constraints [{bone1Name=dyn_gun, bone2Name=dyn_gun_base, swing1LimitDegrees=3, swing2LimitDegrees=3, twistLimitDegrees=0}]
