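# Control Rig for the female medium Bistro Astronaut body that drives helper bones to preserve or bulge muscle volume as joints bend.
# Sequenced two-pass loops cover elbows, knees, hips, glutes, pecs and upper arms, deriving right-side bones by _l/_r name replacement.
# Weighted constraints blend base bones between neighboring bones.
# Clamped remaps of joint rotation variables weight offsets and scales on volume bones, while pecs aim at the upper arms.
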ControlRigBlueprint /BRCosmetics/Characters/Player/Female/Medium/Bodies/F_MED_Bistro_Astronaut/Meshes/F_MED_Bistro_Muscle_Volume_ControlRig
graphs [
 RigVMFunctionLibrary {links [], nodes []}
 RigVMModel {nodes [
   /Script/ControlRig.RigUnit_AimItem(
    DebugSettings=(Scale=10,WorldOffset=(Rotation=(X=0,Y=0,Z=0,W=1),Translation=(X=0,Y=0,Z=0),Scale3D=(X=1,Y=1,Z=1)))
    ExecutePin=()
    Item=(Type=Bone,Name="dyn_ankle_l")
    Primary=(Weight=1,Axis=(X=1,Y=0,Z=0),Target=(X=0,Y=0,Z=0),Kind=Direction,Space=(Type=Bone,Name="ball_l"))
    Secondary=(Weight=1,Axis=(X=1,Y=0,Z=0),Target=(X=0,Y=0,Z=0),Kind=Direction,Space=(Type=Bone))
    Weight=1
   ) {is=struct RigVMUnitNode}
   n1 = &RigUnit_GetTransform(Item=(Type=Bone,Name="ball_l"))
   > &RigUnit_SetTransform(
    Item=(Type=Bone,Name="dyn_ball_l")
    Value=(Rotation=(X=0.0077,Y=-0.3173,Z=-0.0052,W=-0.9483),Translation=(X=0,Y=0,Z=0),Scale3D=(X=1,Y=1,Z=1))
    Weight=1
    Value.Rotation=@n1.Transform.Rotation
    Value.Translation=@n1.Transform.Translation
   )
   &RigUnit_AimItem(
    DebugSettings=(Scale=10,WorldOffset=(Rotation=(X=0,Y=0,Z=0,W=1),Translation=(X=0,Y=0,Z=0),Scale3D=(X=1,Y=1,Z=1)))
    ExecutePin=()
    Item=(Type=Bone,Name="dyn_ankle_r")
    Primary=(Weight=1,Axis=(X=1,Y=0,Z=0),Target=(X=0,Y=0,Z=0),Kind=Direction,Space=(Type=Bone,Name="ball_r"))
    Secondary=(Weight=1,Axis=(X=1,Y=0,Z=0),Target=(X=0,Y=0,Z=0),Kind=Direction,Space=(Type=Bone))
    Weight=1
   )
   > &RigUnit_SetTransform(Item=(Type=Bone,Name="dyn_ball_r"), Weight=1, Value=&RigUnit_GetTransform(Item=(Type=Bone,Name="ball_r")).Transform)
   n2 = /Script/RigVM.RigVMFunction_ForLoopCount(Count=2, ExecutePin=()) {is=struct RigVMUnitNode}
   n3 = /Script/ControlRig.RigUnit_CollectionItems(Items=((Type=Bone,Name="upperarm_twist_01_l"),(Type=Bone,Name="upperarm_twist_02_l"),(Type=Bone,Name="upperarm_l"))) {is=struct RigVMUnitNode}
   n4 = RigVMDispatchNode(
    Values=((),())
    Values.0=@n3.Collection
    Values.1=/Script/ControlRig.RigUnit_CollectionReplaceItems(New=_r, Old=_l, Items=@n3.Collection) {is=struct RigVMUnitNode}.Collection
    Index=@n2.Index
   ) {template="DISPATCH_RigVMDispatch_SelectInt32(in Index,in Values,out Result)"}
   n5 = /Script/ControlRig.RigUnit_CollectionItemAtIndex(Index=1, Collection=@n4.Result) {is=struct RigVMUnitNode}
   n6 = &RigUnit_CollectionItems(Items=((Type=Bone,Name="bicep_in_volume_base_l"),(Type=Bone,Name="bicep_in_volume_l")))
   n7 = RigVMDispatchNode(
    Values=((),())
    Values.0=@n6.Collection
    Values.1=&RigUnit_CollectionReplaceItems(New=_r, Old=_l, Items=@n6.Collection).Collection
    Index=@n2.Index
   ) {template="DISPATCH_RigVMDispatch_SelectInt32(in Index,in Values,out Result)"}
   n8 = &RigUnit_CollectionItemAtIndex(Index=0, Collection=@n7.Result)
   n9 = /Script/ControlRig.RigUnit_ProjectTransformToNewParent(
    Child=(Type=Bone)
    NewParent=(Type=Bone,Name="lowerarm_l")
    OldParent=(Type=Bone,Name="lowerarm_l")
    bChildInitial=True
    bNewParentInitial=false
    bOldParentInitial=True
    NewParent.Name=@n5.Item.Name
    OldParent.Name=@n5.Item.Name
    Child.Name=@n8.Item.Name
   ) {is=struct RigVMUnitNode}
   n10 = &RigUnit_CollectionItemAtIndex(Index=2, Collection=@n4.Result)
   n11 = &RigUnit_ProjectTransformToNewParent(
    Child=(Type=Bone)
    NewParent=(Type=Bone,Name="lowerarm_l")
    OldParent=(Type=Bone,Name="lowerarm_l")
    bChildInitial=True
    bNewParentInitial=false
    bOldParentInitial=True
    NewParent.Name=@n10.Item.Name
    OldParent.Name=@n10.Item.Name
    Child.Name=@n8.Item.Name
   )
   n12 = &RigUnit_CollectionItemAtIndex(Index=0, Collection=@n4.Result)
   n13 = &RigUnit_ProjectTransformToNewParent(
    Child=(Type=Bone)
    NewParent=(Type=Bone,Name="lowerarm_l")
    OldParent=(Type=Bone,Name="lowerarm_l")
    bChildInitial=True
    bNewParentInitial=false
    bOldParentInitial=True
    NewParent.Name=@n12.Item.Name
    OldParent.Name=@n12.Item.Name
    Child.Name=@n8.Item.Name
   )
   > /Script/ControlRig.RigUnit_TransformConstraintPerItem(
    BaseItem=(Type=Bone)
    BaseTransform=(Rotation=(X=0,Y=0,Z=0,W=1),Translation=(X=0,Y=0,Z=0),Scale3D=(X=1,Y=1,Z=1))
    BaseTransformSpace=GlobalSpace
    Item=(Type=Bone,Name="elbow_out_volume_base_l")
    Targets="((Transform=(Rotation=(X=0,Y=0,Z=0,W=1),Translation=(X=0,Y=0,Z=0),Scale3D=(X=1,Y=1,Z=1)),Weight=0.6,bMaintainOffset=false,Filter=(TranslationFilter=(bX=True,bY=True,bZ=True),RotationFilter=(bX=True,bY=True,bZ=True),ScaleFilter=(bX=True,bY=True,bZ=True))),( ...(truncated)"
    bUseInitialTransforms=True
    Targets.1.Transform=@n9.Transform
    Targets.2.Transform=@n11.Transform
    Targets.0.Transform=@n13.Transform
    Item.Name=@n8.Item.Name
   ) {is=struct RigVMUnitNode}
   n14 = RigVMVariableNode() {variable=elbow_rot_l}
   n15 = RigVMVariableNode() {variable=elbow_rot_r}
   n16 = /Script/RigVM.RigVMFunction_MathVectorMul(A=(X=-1.5,Y=1,Z=0), B=(X=1,Y=1,Z=1)) {is=struct RigVMUnitNode, template="Multiply::Execute(in A,in B,out Result)"}
   > &RigUnit_ModifyTransforms(
    ItemToModify=((Item=(Type=Bone,Name="elbow_out_volume_l"),Transform=(Rotation=(X=0,Y=0,Z=0,W=1),Translation=(X=0,Y=-3,Z=0),Scale3D=(X=1,Y=1,Z=1.5))))
    ItemToModify.0.Item.Name=&RigUnit_CollectionItemAtIndex(Index=1, Collection=@n7.Result).Item.Name
    Weight=&RigVMFunction_MathFloatRemap(
     SourceMaximum=90
     bClamp=true
     Value=RigVMDispatchNode(Values=(0,0), Values.0=@n14.Value, Values.1=@n15.Value, Index=@n2.Index) {template="DISPATCH_RigVMDispatch_SelectInt32(in Index,in Values,out Result)"}.Result
    ).Result
    ItemToModify.0.Transform.Translation=RigVMDispatchNode(
     Values=((X=0,Y=8,Z=0),(X=0,Y=-8,Z=0))
     Values.0=@n16.Result
     Values.1=&RigVMFunction_MathVectorMul(B=(X=-1,Y=-1,Z=-1), A=@n16.Result) {template="Multiply::Execute(in A,in B,out Result)"}.Result
     Index=@n2.Index
    ) {template="DISPATCH_RigVMDispatch_SelectInt32(in Index,in Values,out Result)"}.Result
   )
   n17 = &RigVMFunction_ForLoopCount(Count=2, ExecutePin=())
   n18 = &RigUnit_CollectionItems(Items=((Type=Bone,Name="upperarm_twist_01_l"),(Type=Bone,Name="upperarm_twist_02_l"),(Type=Bone,Name="upperarm_l")))
   n19 = RigVMDispatchNode(
    Values=((),())
    Values.0=@n18.Collection
    Values.1=&RigUnit_CollectionReplaceItems(New=_r, Old=_l, Items=@n18.Collection).Collection
    Index=@n17.Index
   ) {template="DISPATCH_RigVMDispatch_SelectInt32(in Index,in Values,out Result)"}
   n20 = &RigUnit_CollectionItemAtIndex(Index=2, Collection=@n19.Result)
   n21 = &RigUnit_CollectionItems(Items=((Type=Bone,Name="bicep_out_volume_base_l"),(Type=Bone,Name="bicep_out_volume_l")))
   n22 = RigVMDispatchNode(
    Values=((),())
    Values.0=@n21.Collection
    Values.1=&RigUnit_CollectionReplaceItems(New=_r, Old=_l, Items=@n21.Collection).Collection
    Index=@n17.Index
   ) {template="DISPATCH_RigVMDispatch_SelectInt32(in Index,in Values,out Result)"}
   n23 = &RigUnit_CollectionItemAtIndex(Index=0, Collection=@n22.Result)
   n24 = &RigUnit_ProjectTransformToNewParent(
    Child=(Type=Bone)
    NewParent=(Type=Bone,Name="lowerarm_l")
    OldParent=(Type=Bone,Name="lowerarm_l")
    bChildInitial=True
    bNewParentInitial=false
    bOldParentInitial=True
    NewParent.Name=@n20.Item.Name
    OldParent.Name=@n20.Item.Name
    Child.Name=@n23.Item.Name
   )
   n25 = &RigUnit_CollectionItemAtIndex(Index=1, Collection=@n19.Result)
   n26 = &RigUnit_ProjectTransformToNewParent(
    Child=(Type=Bone)
    NewParent=(Type=Bone,Name="lowerarm_l")
    OldParent=(Type=Bone,Name="lowerarm_l")
    bChildInitial=True
    bNewParentInitial=false
    bOldParentInitial=True
    NewParent.Name=@n25.Item.Name
    OldParent.Name=@n25.Item.Name
    Child.Name=@n23.Item.Name
   )
   n27 = &RigUnit_CollectionItemAtIndex(Index=0, Collection=@n19.Result)
   n28 = &RigUnit_ProjectTransformToNewParent(
    Child=(Type=Bone)
    NewParent=(Type=Bone,Name="lowerarm_l")
    OldParent=(Type=Bone,Name="lowerarm_l")
    bChildInitial=True
    bNewParentInitial=false
    bOldParentInitial=True
    NewParent.Name=@n27.Item.Name
    OldParent.Name=@n27.Item.Name
    Child.Name=@n23.Item.Name
   )
   > &RigUnit_TransformConstraintPerItem(
    BaseItem=(Type=Bone)
    BaseTransform=(Rotation=(X=0,Y=0,Z=0,W=1),Translation=(X=0,Y=0,Z=0),Scale3D=(X=1,Y=1,Z=1))
    BaseTransformSpace=GlobalSpace
    Item=(Type=Bone,Name="elbow_out_volume_base_l")
    Targets="((Transform=(Rotation=(X=0,Y=0,Z=0,W=1),Translation=(X=0,Y=0,Z=0),Scale3D=(X=1,Y=1,Z=1)),Weight=0.6,bMaintainOffset=false,Filter=(TranslationFilter=(bX=True,bY=True,bZ=True),RotationFilter=(bX=True,bY=True,bZ=True),ScaleFilter=(bX=True,bY=True,bZ=True))),( ...(truncated)"
    bUseInitialTransforms=True
    Targets.2.Transform=@n24.Transform
    Targets.1.Transform=@n26.Transform
    Targets.0.Transform=@n28.Transform
    Item.Name=@n23.Item.Name
   )
   n29 = RigVMVariableNode() {variable=elbow_rot_l}
   n30 = RigVMVariableNode() {variable=elbow_rot_r}
   n31 = &RigVMFunction_MathVectorMul(A=(X=1.5,Y=-0.5,Z=0), B=(X=1,Y=1,Z=1)) {template="Multiply::Execute(in A,in B,out Result)"}
   > &RigUnit_ModifyTransforms(
    ItemToModify=((Item=(Type=Bone,Name="elbow_out_volume_l"),Transform=(Rotation=(X=0,Y=0,Z=0,W=1),Translation=(X=0,Y=-3,Z=0),Scale3D=(X=1,Y=1,Z=1.5))))
    ItemToModify.0.Item.Name=&RigUnit_CollectionItemAtIndex(Index=1, Collection=@n22.Result).Item.Name
    Weight=&RigVMFunction_MathFloatRemap(
     SourceMaximum=90
     bClamp=true
     Value=RigVMDispatchNode(Values=(0,0), Values.0=@n29.Value, Values.1=@n30.Value, Index=@n17.Index) {template="DISPATCH_RigVMDispatch_SelectInt32(in Index,in Values,out Result)"}.Result
    ).Result
    ItemToModify.0.Transform.Translation=RigVMDispatchNode(
     Values=((X=0,Y=8,Z=0),(X=0,Y=-8,Z=0))
     Values.0=@n31.Result
     Values.1=&RigVMFunction_MathVectorMul(B=(X=-1,Y=-1,Z=-1), A=@n31.Result) {template="Multiply::Execute(in A,in B,out Result)"}.Result
     Index=@n17.Index
    ) {template="DISPATCH_RigVMDispatch_SelectInt32(in Index,in Values,out Result)"}.Result
   )
   n32 = &RigUnit_BeginExecution()
   n33 = /Script/ControlRig.RigUnit_SequenceExecution(ExecuteContext=@n32.ExecutePin) {is=struct RigVMUnitNode}
   n34 = &RigUnit_SequenceExecution(ExecuteContext=@n33.A)
   n35 = &RigVMFunction_ForLoopCount(Count=2, ExecutePin=@n34.A)
   n36 = &RigUnit_CollectionItems(Items=((Type=Bone,Name="upperarm_l"),(Type=Bone,Name="lowerarm_l"),(Type=Bone,Name="elbow_out_volume_base_l")))
   n37 = RigVMDispatchNode(
    Values=((),())
    Values.0=@n36.Collection
    Values.1=&RigUnit_CollectionReplaceItems(New=_r, Old=_l, Items=@n36.Collection).Collection
    Index=@n35.Index
   ) {template="DISPATCH_RigVMDispatch_SelectInt32(in Index,in Values,out Result)"}
   n38 = &RigUnit_CollectionItemAtIndex(Index=2, Collection=@n37.Result)
   n39 = &RigUnit_CollectionItemAtIndex(Index=0, Collection=@n37.Result)
   n40 = &RigUnit_ProjectTransformToNewParent(
    Child=(Type=Bone)
    NewParent=(Type=Bone,Name="lowerarm_l")
    OldParent=(Type=Bone,Name="lowerarm_l")
    bChildInitial=True
    bOldParentInitial=True
    Child.Name=@n38.Item.Name
    NewParent.Name=@n39.Item.Name
    OldParent.Name=@n39.Item.Name
   )
   n41 = &RigUnit_CollectionItemAtIndex(Index=1, Collection=@n37.Result)
   n42 = &RigUnit_ProjectTransformToNewParent(
    Child=(Type=Bone)
    NewParent=(Type=Bone,Name="lowerarm_l")
    OldParent=(Type=Bone,Name="lowerarm_l")
    bChildInitial=True
    bOldParentInitial=True
    NewParent.Name=@n41.Item.Name
    OldParent.Name=@n41.Item.Name
    Child.Name=@n38.Item.Name
   )
   > &RigUnit_TransformConstraintPerItem(
    BaseItem=(Type=Bone)
    BaseTransform=(Rotation=(X=0,Y=0,Z=0,W=1),Translation=(X=0,Y=0,Z=0),Scale3D=(X=1,Y=1,Z=1))
    BaseTransformSpace=GlobalSpace
    Item=(Type=Bone,Name="elbow_out_volume_base_l")
    Targets="((Transform=(Rotation=(X=0,Y=0,Z=0,W=1),Translation=(X=0,Y=0,Z=0),Scale3D=(X=1,Y=1,Z=1)),Weight=0.7,bMaintainOffset=false,Filter=(TranslationFilter=(bX=True,bY=True,bZ=True),RotationFilter=(bX=True,bY=True,bZ=True),ScaleFilter=(bX=True,bY=True,bZ=True))),( ...(truncated)"
    bUseInitialTransforms=True
    Targets.0.Transform=@n40.Transform
    Targets.1.Transform=@n42.Transform
    Item.Name=@n38.Item.Name
   )
   n43 = &RigUnit_CollectionItems(Items=((Type=Bone,Name="elbow_out_volume_l")))
   n44 = &RigUnit_CollectionReplaceItems(New=_r, Old=_l, Items=@n43.Collection)
   n45 = RigVMVariableNode() {variable=elbow_rot_l}
   n46 = RigVMVariableNode() {variable=elbow_rot_r}
   n47 = &RigVMFunction_MathVectorMul(A=(X=0,Y=-1,Z=0), B=(X=1,Y=1,Z=1)) {template="Multiply::Execute(in A,in B,out Result)"}
   > &RigUnit_ModifyTransforms(
    ItemToModify=((Item=(Type=Bone,Name="elbow_out_volume_l"),Transform=(Rotation=(X=0,Y=0,Z=0,W=1),Translation=(X=0,Y=-3,Z=0),Scale3D=(X=0.6,Y=1,Z=1))))
    ItemToModify.0.Item.Name=&RigUnit_CollectionItemAtIndex(
     Index=0
     Collection=RigVMDispatchNode(Values=((),()), Values.0=@n43.Collection, Values.1=@n44.Collection, Index=@n35.Index) {template="DISPATCH_RigVMDispatch_SelectInt32(in Index,in Values,out Result)"}.Result
    ).Item.Name
    Weight=&RigVMFunction_MathFloatRemap(
     SourceMaximum=80
     bClamp=true
     Value=RigVMDispatchNode(Values=(0,0), Values.0=@n45.Value, Values.1=@n46.Value, Index=@n35.Index) {template="DISPATCH_RigVMDispatch_SelectInt32(in Index,in Values,out Result)"}.Result
    ).Result
    ItemToModify.0.Transform.Translation=RigVMDispatchNode(
     Values=((X=0,Y=-3,Z=0),(X=0,Y=3,Z=0))
     Values.1=&RigVMFunction_MathVectorMul(B=(X=-1,Y=-1,Z=-1), A=@n47.Result) {template="Multiply::Execute(in A,in B,out Result)"}.Result
     Values.0=@n47.Result
     Index=@n35.Index
    ) {template="DISPATCH_RigVMDispatch_SelectInt32(in Index,in Values,out Result)"}.Result
   )
   n48 = &RigVMFunction_ForLoopCount(Count=2, ExecutePin=@n34.B)
   n49 = &RigUnit_CollectionItems(Items=((Type=Bone,Name="upperarm_l"),(Type=Bone,Name="lowerarm_l"),(Type=Bone,Name="elbow_in_volume_base_l")))
   n50 = RigVMDispatchNode(
    Values=((),())
    Values.0=@n49.Collection
    Values.1=&RigUnit_CollectionReplaceItems(New=_r, Old=_l, Items=@n49.Collection).Collection
    Index=@n48.Index
   ) {template="DISPATCH_RigVMDispatch_SelectInt32(in Index,in Values,out Result)"}
   n51 = &RigUnit_CollectionItemAtIndex(Index=2, Collection=@n50.Result)
   n52 = &RigUnit_CollectionItemAtIndex(Index=1, Collection=@n50.Result)
   n53 = &RigUnit_ProjectTransformToNewParent(
    Child=(Type=Bone)
    NewParent=(Type=Bone,Name="lowerarm_l")
    OldParent=(Type=Bone,Name="lowerarm_l")
    bChildInitial=True
    bOldParentInitial=True
    Child.Name=@n51.Item.Name
    NewParent.Name=@n52.Item.Name
    OldParent.Name=@n52.Item.Name
   )
   n54 = &RigUnit_CollectionItemAtIndex(Index=0, Collection=@n50.Result)
   n55 = &RigUnit_ProjectTransformToNewParent(
    Child=(Type=Bone)
    NewParent=(Type=Bone,Name="lowerarm_l")
    OldParent=(Type=Bone,Name="lowerarm_l")
    bChildInitial=True
    bOldParentInitial=True
    Child.Name=@n51.Item.Name
    NewParent.Name=@n54.Item.Name
    OldParent.Name=@n54.Item.Name
   )
   > &RigUnit_TransformConstraintPerItem(
    BaseItem=(Type=Bone)
    BaseTransform=(Rotation=(X=0,Y=0,Z=0,W=1),Translation=(X=0,Y=0,Z=0),Scale3D=(X=1,Y=1,Z=1))
    BaseTransformSpace=GlobalSpace
    Item=(Type=Bone,Name="elbow_out_volume_base_l")
    Targets="((Transform=(Rotation=(X=0,Y=0,Z=0,W=1),Translation=(X=0,Y=0,Z=0),Scale3D=(X=1,Y=1,Z=1)),Weight=0.7,bMaintainOffset=false,Filter=(TranslationFilter=(bX=True,bY=True,bZ=True),RotationFilter=(bX=True,bY=True,bZ=True),ScaleFilter=(bX=True,bY=True,bZ=True))),( ...(truncated)"
    bUseInitialTransforms=True
    Targets.1.Transform=@n53.Transform
    Targets.0.Transform=@n55.Transform
    Item.Name=@n51.Item.Name
   )
   n56 = &RigUnit_CollectionItems(Items=((Type=Bone,Name="elbow_in_volume_l")))
   n57 = &RigUnit_CollectionReplaceItems(New=_r, Old=_l, Items=@n56.Collection)
   n58 = RigVMVariableNode() {variable=elbow_rot_l}
   n59 = RigVMVariableNode() {variable=elbow_rot_r}
   n60 = &RigVMFunction_MathVectorMul(A=(X=0,Y=5,Z=0), B=(X=1,Y=1,Z=1)) {template="Multiply::Execute(in A,in B,out Result)"}
   > &RigUnit_ModifyTransforms(
    ItemToModify=((Item=(Type=Bone,Name="elbow_out_volume_l"),Transform=(Rotation=(X=0,Y=0,Z=0,W=1),Translation=(X=0,Y=-3,Z=0),Scale3D=(X=0.5,Y=1,Z=1))))
    ItemToModify.0.Item.Name=&RigUnit_CollectionItemAtIndex(
     Index=0
     Collection=RigVMDispatchNode(Values=((),()), Values.0=@n56.Collection, Values.1=@n57.Collection, Index=@n48.Index) {template="DISPATCH_RigVMDispatch_SelectInt32(in Index,in Values,out Result)"}.Result
    ).Item.Name
    Weight=&RigVMFunction_MathFloatRemap(
     SourceMaximum=90
     SourceMinimum=20
     bClamp=true
     Value=RigVMDispatchNode(Values=(0,0), Values.0=@n58.Value, Values.1=@n59.Value, Index=@n48.Index) {template="DISPATCH_RigVMDispatch_SelectInt32(in Index,in Values,out Result)"}.Result
    ).Result
    ItemToModify.0.Transform.Translation=RigVMDispatchNode(
     Values=((X=0,Y=-3,Z=0),(X=0,Y=3,Z=0))
     Values.0=@n60.Result
     Values.1=&RigVMFunction_MathVectorMul(B=(X=-1,Y=-1,Z=-1), A=@n60.Result) {template="Multiply::Execute(in A,in B,out Result)"}.Result
     Index=@n48.Index
    ) {template="DISPATCH_RigVMDispatch_SelectInt32(in Index,in Values,out Result)"}.Result
   )
   n61 = &RigUnit_SequenceExecution(ExecuteContext=@n34.C)
   n62 = &RigVMFunction_ForLoopCount(Count=2, ExecutePin=@n61.A)
   n63 = &RigUnit_CollectionItems(Items=((Type=Bone,Name="thigh_l"),(Type=Bone,Name="calf_l"),(Type=Bone,Name="knee_out_volume_base_l")))
   n64 = RigVMDispatchNode(
    Values=((),())
    Values.0=@n63.Collection
    Values.1=&RigUnit_CollectionReplaceItems(New=_r, Old=_l, Items=@n63.Collection).Collection
    Index=@n62.Index
   ) {template="DISPATCH_RigVMDispatch_SelectInt32(in Index,in Values,out Result)"}
   n65 = &RigUnit_CollectionItemAtIndex(Index=2, Collection=@n64.Result)
   n66 = &RigUnit_CollectionItemAtIndex(Index=1, Collection=@n64.Result)
   n67 = &RigUnit_ProjectTransformToNewParent(
    Child=(Type=Bone)
    NewParent=(Type=Bone,Name="lowerarm_l")
    OldParent=(Type=Bone,Name="lowerarm_l")
    bChildInitial=True
    bOldParentInitial=True
    Child.Name=@n65.Item.Name
    NewParent.Name=@n66.Item.Name
    OldParent.Name=@n66.Item.Name
   )
   n68 = &RigUnit_CollectionItemAtIndex(Index=0, Collection=@n64.Result)
   n69 = &RigUnit_ProjectTransformToNewParent(
    Child=(Type=Bone)
    NewParent=(Type=Bone,Name="lowerarm_l")
    OldParent=(Type=Bone,Name="lowerarm_l")
    bChildInitial=True
    bOldParentInitial=True
    Child.Name=@n65.Item.Name
    NewParent.Name=@n68.Item.Name
    OldParent.Name=@n68.Item.Name
   )
   > &RigUnit_TransformConstraintPerItem(
    BaseItem=(Type=Bone)
    BaseTransform=(Rotation=(X=0,Y=0,Z=0,W=1),Translation=(X=0,Y=0,Z=0),Scale3D=(X=1,Y=1,Z=1))
    BaseTransformSpace=GlobalSpace
    Item=(Type=Bone,Name="elbow_out_volume_base_l")
    Targets="((Transform=(Rotation=(X=0,Y=0,Z=0,W=1),Translation=(X=0,Y=0,Z=0),Scale3D=(X=1,Y=1,Z=1)),Weight=0.8,bMaintainOffset=false,Filter=(TranslationFilter=(bX=True,bY=True,bZ=True),RotationFilter=(bX=True,bY=True,bZ=True),ScaleFilter=(bX=True,bY=True,bZ=True))),( ...(truncated)"
    bUseInitialTransforms=True
    Targets.1.Transform=@n67.Transform
    Targets.0.Transform=@n69.Transform
    Item.Name=@n65.Item.Name
   )
   n70 = &RigUnit_CollectionItems(Items=((Type=Bone,Name="knee_out_volume_l")))
   n71 = &RigUnit_CollectionReplaceItems(New=_r, Old=_l, Items=@n70.Collection)
   n72 = RigVMVariableNode() {variable=knee_rot_l}
   n73 = RigVMVariableNode() {variable=knee_rot_r}
   n74 = &RigVMFunction_MathVectorMul(A=(X=0,Y=3,Z=0), B=(X=1,Y=1,Z=1)) {template="Multiply::Execute(in A,in B,out Result)"}
   > &RigUnit_ModifyTransforms(
    ItemToModify=((Item=(Type=Bone,Name="elbow_out_volume_l"),Transform=(Rotation=(X=0,Y=0,Z=0,W=1),Translation=(X=0,Y=-3,Z=0),Scale3D=(X=1.5,Y=1,Z=1.5))))
    ItemToModify.0.Item.Name=&RigUnit_CollectionItemAtIndex(
     Index=0
     Collection=RigVMDispatchNode(Values=((),()), Values.0=@n70.Collection, Values.1=@n71.Collection, Index=@n62.Index) {template="DISPATCH_RigVMDispatch_SelectInt32(in Index,in Values,out Result)"}.Result
    ).Item.Name
    Weight=&RigVMFunction_MathFloatRemap(
     SourceMaximum=80
     SourceMinimum=0
     bClamp=true
     Value=RigVMDispatchNode(Values=(0,0), Values.0=@n72.Value, Values.1=@n73.Value, Index=@n62.Index) {template="DISPATCH_RigVMDispatch_SelectInt32(in Index,in Values,out Result)"}.Result
    ).Result
    ItemToModify.0.Transform.Translation=RigVMDispatchNode(
     Values=((X=0,Y=-3,Z=0),(X=0,Y=3,Z=0))
     Values.0=@n74.Result
     Values.1=&RigVMFunction_MathVectorMul(B=(X=-1,Y=-1,Z=-1), A=@n74.Result) {template="Multiply::Execute(in A,in B,out Result)"}.Result
     Index=@n62.Index
    ) {template="DISPATCH_RigVMDispatch_SelectInt32(in Index,in Values,out Result)"}.Result
   )
   n75 = &RigVMFunction_ForLoopCount(Count=2, ExecutePin=@n61.B)
   n76 = &RigUnit_CollectionItems(Items=((Type=Bone,Name="thigh_l"),(Type=Bone,Name="calf_l"),(Type=Bone,Name="knee_in_volume_base_l")))
   n77 = RigVMDispatchNode(
    Values=((),())
    Values.0=@n76.Collection
    Values.1=&RigUnit_CollectionReplaceItems(New=_r, Old=_l, Items=@n76.Collection).Collection
    Index=@n75.Index
   ) {template="DISPATCH_RigVMDispatch_SelectInt32(in Index,in Values,out Result)"}
   n78 = &RigUnit_CollectionItemAtIndex(Index=2, Collection=@n77.Result)
   n79 = &RigUnit_CollectionItemAtIndex(Index=1, Collection=@n77.Result)
   n80 = &RigUnit_ProjectTransformToNewParent(
    Child=(Type=Bone)
    NewParent=(Type=Bone,Name="lowerarm_l")
    OldParent=(Type=Bone,Name="lowerarm_l")
    bChildInitial=True
    bOldParentInitial=True
    Child.Name=@n78.Item.Name
    NewParent.Name=@n79.Item.Name
    OldParent.Name=@n79.Item.Name
   )
   n81 = &RigUnit_CollectionItemAtIndex(Index=0, Collection=@n77.Result)
   n82 = &RigUnit_ProjectTransformToNewParent(
    Child=(Type=Bone)
    NewParent=(Type=Bone,Name="lowerarm_l")
    OldParent=(Type=Bone,Name="lowerarm_l")
    bChildInitial=True
    bOldParentInitial=True
    Child.Name=@n78.Item.Name
    NewParent.Name=@n81.Item.Name
    OldParent.Name=@n81.Item.Name
   )
   > &RigUnit_TransformConstraintPerItem(
    BaseItem=(Type=Bone)
    BaseTransform=(Rotation=(X=0,Y=0,Z=0,W=1),Translation=(X=0,Y=0,Z=0),Scale3D=(X=1,Y=1,Z=1))
    BaseTransformSpace=GlobalSpace
    Item=(Type=Bone,Name="elbow_out_volume_base_l")
    Targets="((Transform=(Rotation=(X=0,Y=0,Z=0,W=1),Translation=(X=0,Y=0,Z=0),Scale3D=(X=1,Y=1,Z=1)),Weight=0.6,bMaintainOffset=false,Filter=(TranslationFilter=(bX=True,bY=True,bZ=True),RotationFilter=(bX=True,bY=True,bZ=True),ScaleFilter=(bX=True,bY=True,bZ=True))),( ...(truncated)"
    bUseInitialTransforms=True
    Targets.1.Transform=@n80.Transform
    Targets.0.Transform=@n82.Transform
    Item.Name=@n78.Item.Name
   )
   n83 = &RigUnit_CollectionItems(Items=((Type=Bone,Name="knee_in_volume_l")))
   n84 = &RigUnit_CollectionReplaceItems(New=_r, Old=_l, Items=@n83.Collection)
   n85 = RigVMVariableNode() {variable=knee_rot_l}
   n86 = RigVMVariableNode() {variable=knee_rot_r}
   n87 = &RigVMFunction_MathVectorMul(A=(X=0,Y=-10,Z=0), B=(X=1,Y=1,Z=1)) {template="Multiply::Execute(in A,in B,out Result)"}
   > &RigUnit_ModifyTransforms(
    ItemToModify=((Item=(Type=Bone,Name="elbow_out_volume_l"),Transform=(Rotation=(X=0,Y=0,Z=0,W=1),Translation=(X=0,Y=-3,Z=0),Scale3D=(X=0.5,Y=1,Z=1))))
    ItemToModify.0.Item.Name=&RigUnit_CollectionItemAtIndex(
     Index=0
     Collection=RigVMDispatchNode(Values=((),()), Values.0=@n83.Collection, Values.1=@n84.Collection, Index=@n75.Index) {template="DISPATCH_RigVMDispatch_SelectInt32(in Index,in Values,out Result)"}.Result
    ).Item.Name
    Weight=&RigVMFunction_MathFloatRemap(
     SourceMaximum=100
     SourceMinimum=0
     bClamp=true
     Value=RigVMDispatchNode(Values=(0,0), Values.0=@n85.Value, Values.1=@n86.Value, Index=@n75.Index) {template="DISPATCH_RigVMDispatch_SelectInt32(in Index,in Values,out Result)"}.Result
    ).Result
    ItemToModify.0.Transform.Translation=RigVMDispatchNode(
     Values=((X=0,Y=-3,Z=0),(X=0,Y=3,Z=0))
     Values.0=@n87.Result
     Values.1=&RigVMFunction_MathVectorMul(B=(X=-1,Y=-1,Z=-1), A=@n87.Result) {template="Multiply::Execute(in A,in B,out Result)"}.Result
     Index=@n75.Index
    ) {template="DISPATCH_RigVMDispatch_SelectInt32(in Index,in Values,out Result)"}.Result
   )
   n88 = &RigUnit_SequenceExecution(ExecuteContext=@n33.B)
   n89 = &RigVMFunction_ForLoopCount(Count=2, ExecutePin=@n88.A)
   n90 = &RigUnit_CollectionItems(Items=((Type=Bone,Name="pelvis"),(Type=Bone,Name="thigh_l"),(Type=Bone,Name="crouch_volume_base_l")))
   n91 = RigVMDispatchNode(
    Values=((),())
    Values.0=@n90.Collection
    Values.1=&RigUnit_CollectionReplaceItems(New=_r, Old=_l, Items=@n90.Collection).Collection
    Index=@n89.Index
   ) {template="DISPATCH_RigVMDispatch_SelectInt32(in Index,in Values,out Result)"}
   n92 = &RigUnit_CollectionItemAtIndex(Index=2, Collection=@n91.Result)
   n93 = &RigUnit_CollectionItemAtIndex(Index=1, Collection=@n91.Result)
   n94 = &RigUnit_ProjectTransformToNewParent(
    Child=(Type=Bone)
    NewParent=(Type=Bone,Name="lowerarm_l")
    OldParent=(Type=Bone,Name="lowerarm_l")
    bChildInitial=True
    bOldParentInitial=True
    Child.Name=@n92.Item.Name
    NewParent.Name=@n93.Item.Name
    OldParent.Name=@n93.Item.Name
   )
   n95 = &RigUnit_CollectionItemAtIndex(Index=0, Collection=@n91.Result)
   n96 = &RigUnit_ProjectTransformToNewParent(
    Child=(Type=Bone)
    NewParent=(Type=Bone,Name="lowerarm_l")
    OldParent=(Type=Bone,Name="lowerarm_l")
    bChildInitial=True
    bOldParentInitial=True
    Child.Name=@n92.Item.Name
    NewParent.Name=@n95.Item.Name
    OldParent.Name=@n95.Item.Name
   )
   > &RigUnit_TransformConstraintPerItem(
    BaseItem=(Type=Bone)
    BaseTransform=(Rotation=(X=0,Y=0,Z=0,W=1),Translation=(X=0,Y=0,Z=0),Scale3D=(X=1,Y=1,Z=1))
    BaseTransformSpace=GlobalSpace
    Item=(Type=Bone,Name="elbow_out_volume_base_l")
    Targets="((Transform=(Rotation=(X=0,Y=0,Z=0,W=1),Translation=(X=0,Y=0,Z=0),Scale3D=(X=1,Y=1,Z=1)),Weight=0.3,bMaintainOffset=false,Filter=(TranslationFilter=(bX=True,bY=True,bZ=True),RotationFilter=(bX=True,bY=True,bZ=True),ScaleFilter=(bX=True,bY=True,bZ=True))),( ...(truncated)"
    bUseInitialTransforms=True
    Targets.1.Transform=@n94.Transform
    Targets.0.Transform=@n96.Transform
    Item.Name=@n92.Item.Name
   )
   n97 = &RigUnit_CollectionItems(Items=((Type=Bone,Name="crouch_volume_l")))
   n98 = &RigUnit_CollectionReplaceItems(New=_r, Old=_l, Items=@n97.Collection)
   n99 = RigVMVariableNode() {variable=thigh_rot_l}
   n100 = RigVMVariableNode() {variable=thigh_rot_r}
   n101 = &RigVMFunction_MathVectorMul(A=(X=-3,Y=12,Z=0), B=(X=1,Y=1,Z=1)) {template="Multiply::Execute(in A,in B,out Result)"}
   > &RigUnit_ModifyTransforms(
    ItemToModify=((Item=(Type=Bone,Name="elbow_out_volume_l"),Transform=(Rotation=(X=0,Y=0,Z=0,W=1),Translation=(X=0,Y=-3,Z=0),Scale3D=(X=0.6,Y=1,Z=1))))
    ItemToModify.0.Item.Name=&RigUnit_CollectionItemAtIndex(
     Index=0
     Collection=RigVMDispatchNode(Values=((),()), Values.0=@n97.Collection, Values.1=@n98.Collection, Index=@n89.Index) {template="DISPATCH_RigVMDispatch_SelectInt32(in Index,in Values,out Result)"}.Result
    ).Item.Name
    Weight=&RigVMFunction_MathFloatRemap(
     SourceMaximum=0
     SourceMinimum=-90
     TargetMaximum=0
     TargetMinimum=1
     bClamp=true
     Value=RigVMDispatchNode(Values=(0,0), Values.0=@n99.Value, Values.1=@n100.Value, Index=@n89.Index) {template="DISPATCH_RigVMDispatch_SelectInt32(in Index,in Values,out Result)"}.Result
    ).Result
    ItemToModify.0.Transform.Translation=RigVMDispatchNode(
     Values=((X=0,Y=-3,Z=0),(X=0,Y=3,Z=0))
     Values.0=@n101.Result
     Values.1=&RigVMFunction_MathVectorMul(B=(X=-1,Y=-1,Z=-1), A=@n101.Result) {template="Multiply::Execute(in A,in B,out Result)"}.Result
     Index=@n89.Index
    ) {template="DISPATCH_RigVMDispatch_SelectInt32(in Index,in Values,out Result)"}.Result
   )
   n102 = &RigVMFunction_ForLoopCount(Count=2, ExecutePin=@n88.B)
   n103 = &RigUnit_CollectionItems(Items=((Type=Bone,Name="pelvis"),(Type=Bone,Name="thigh_l"),(Type=Bone,Name="glute_volume_base_l")))
   n104 = RigVMDispatchNode(
    Values=((),())
    Values.0=@n103.Collection
    Values.1=&RigUnit_CollectionReplaceItems(New=_r, Old=_l, Items=@n103.Collection).Collection
    Index=@n102.Index
   ) {template="DISPATCH_RigVMDispatch_SelectInt32(in Index,in Values,out Result)"}
   n105 = &RigUnit_CollectionItemAtIndex(Index=2, Collection=@n104.Result)
   n106 = &RigUnit_CollectionItemAtIndex(Index=1, Collection=@n104.Result)
   n107 = &RigUnit_ProjectTransformToNewParent(
    Child=(Type=Bone)
    NewParent=(Type=Bone,Name="lowerarm_l")
    OldParent=(Type=Bone,Name="lowerarm_l")
    bChildInitial=True
    bOldParentInitial=True
    Child.Name=@n105.Item.Name
    NewParent.Name=@n106.Item.Name
    OldParent.Name=@n106.Item.Name
   )
   n108 = &RigUnit_CollectionItemAtIndex(Index=0, Collection=@n104.Result)
   n109 = &RigUnit_ProjectTransformToNewParent(
    Child=(Type=Bone)
    NewParent=(Type=Bone,Name="lowerarm_l")
    OldParent=(Type=Bone,Name="lowerarm_l")
    bChildInitial=True
    bOldParentInitial=True
    Child.Name=@n105.Item.Name
    NewParent.Name=@n108.Item.Name
    OldParent.Name=@n108.Item.Name
   )
   > &RigUnit_TransformConstraintPerItem(
    BaseItem=(Type=Bone)
    BaseTransform=(Rotation=(X=0,Y=0,Z=0,W=1),Translation=(X=0,Y=0,Z=0),Scale3D=(X=1,Y=1,Z=1))
    BaseTransformSpace=GlobalSpace
    Item=(Type=Bone,Name="elbow_out_volume_base_l")
    Targets="((Transform=(Rotation=(X=0,Y=0,Z=0,W=1),Translation=(X=0,Y=0,Z=0),Scale3D=(X=1,Y=1,Z=1)),Weight=0.9,bMaintainOffset=false,Filter=(TranslationFilter=(bX=True,bY=True,bZ=True),RotationFilter=(bX=True,bY=True,bZ=True),ScaleFilter=(bX=True,bY=True,bZ=True))),( ...(truncated)"
    bUseInitialTransforms=True
    Targets.1.Transform=@n107.Transform
    Targets.0.Transform=@n109.Transform
    Item.Name=@n105.Item.Name
   )
   n110 = &RigUnit_CollectionItems(Items=((Type=Bone,Name="glute_volume_l")))
   n111 = &RigUnit_CollectionItemAtIndex(
    Index=0
    Collection=RigVMDispatchNode(
     Values=((),())
     Values.0=@n110.Collection
     Values.1=&RigUnit_CollectionReplaceItems(New=_r, Old=_l, Items=@n110.Collection).Collection
     Index=@n102.Index
    ) {template="DISPATCH_RigVMDispatch_SelectInt32(in Index,in Values,out Result)"}.Result
   )
   n112 = &RigVMFunction_MathFloatRemap(
    SourceMaximum=0
    SourceMinimum=-100
    TargetMaximum=0
    TargetMinimum=1
    bClamp=true
    Value=RigVMDispatchNode(
     Values=(0,0)
     Values.0=RigVMVariableNode() {variable=thigh_rot_l}.Value
     Values.1=RigVMVariableNode() {variable=thigh_rot_r}.Value
     Index=@n102.Index
    ) {template="DISPATCH_RigVMDispatch_SelectInt32(in Index,in Values,out Result)"}.Result
   )
   n113 = &RigVMFunction_MathVectorMul(A=(X=-5,Y=0,Z=0), B=(X=1,Y=1,Z=1)) {template="Multiply::Execute(in A,in B,out Result)"}
   > &RigUnit_ModifyTransforms(
    ItemToModify=((Item=(Type=Bone,Name="elbow_out_volume_l"),Transform=(Rotation=(X=0,Y=0,Z=0,W=1),Translation=(X=0,Y=-3,Z=0),Scale3D=(X=1,Y=1,Z=1))))
    ItemToModify.0.Item.Name=@n111.Item.Name
    Weight=@n112.Result
    ItemToModify.0.Transform.Translation=RigVMDispatchNode(
     Values=((X=0,Y=-3,Z=0),(X=0,Y=3,Z=0))
     Values.0=@n113.Result
     Values.1=&RigVMFunction_MathVectorMul(B=(X=-1,Y=-1,Z=-1), A=@n113.Result) {template="Multiply::Execute(in A,in B,out Result)"}.Result
     Index=@n102.Index
    ) {template="DISPATCH_RigVMDispatch_SelectInt32(in Index,in Values,out Result)"}.Result
   )
   n114 = &RigVMFunction_MathVectorMul(A=(X=0,Y=-4,Z=0), B=(X=1,Y=1,Z=1)) {template="Multiply::Execute(in A,in B,out Result)"}
   > &RigUnit_ModifyTransforms(
    ItemToModify=((Item=(Type=Bone,Name="elbow_out_volume_l"),Transform=(Rotation=(X=0,Y=0,Z=0,W=1),Translation=(X=0,Y=-3,Z=0),Scale3D=(X=1,Y=1,Z=1))))
    Weight=&RigVMFunction_MathFloatRemap(SourceMaximum=50, SourceMinimum=0, TargetMinimum=0, bClamp=true, Value=@n112.Result).Result
    ItemToModify.0.Transform.Translation=RigVMDispatchNode(
     Values=((X=0,Y=8,Z=0),(X=0,Y=-8,Z=0))
     Values.0=@n114.Result
     Values.1=&RigVMFunction_MathVectorMul(B=(X=-1,Y=-1,Z=-1), A=@n114.Result) {template="Multiply::Execute(in A,in B,out Result)"}.Result
     Index=@n102.Index
    ) {template="DISPATCH_RigVMDispatch_SelectInt32(in Index,in Values,out Result)"}.Result
    ItemToModify.0.Item.Name=@n111.Item.Name
   )
   n115 = &RigVMFunction_ForLoopCount(Count=2, ExecutePin=@n88.C)
   n116 = &RigUnit_CollectionItems(Items=((Type=Bone,Name="spine_03"),(Type=Bone,Name="spine_04"),(Type=Bone,Name="spine_05"),(Name="pec_volume_aim_base_l")))
   n117 = RigVMDispatchNode(
    Values=((),())
    Values.0=@n116.Collection
    Values.1=&RigUnit_CollectionReplaceItems(New=_r, Old=_l, Items=@n116.Collection).Collection
    Index=@n115.Index
   ) {template="DISPATCH_RigVMDispatch_SelectInt32(in Index,in Values,out Result)"}
   n118 = &RigUnit_CollectionItemAtIndex(Index=1, Collection=@n117.Result)
   n119 = &RigUnit_CollectionItemAtIndex(Index=3, Collection=@n117.Result)
   n120 = &RigUnit_ProjectTransformToNewParent(
    Child=(Type=Bone)
    NewParent=(Type=Bone,Name="lowerarm_l")
    OldParent=(Type=Bone,Name="lowerarm_l")
    bChildInitial=True
    bNewParentInitial=false
    bOldParentInitial=True
    NewParent.Name=@n118.Item.Name
    OldParent.Name=@n118.Item.Name
    Child.Name=@n119.Item.Name
   )
   n121 = &RigUnit_CollectionItemAtIndex(Index=2, Collection=@n117.Result)
   n122 = &RigUnit_ProjectTransformToNewParent(
    Child=(Type=Bone)
    NewParent=(Type=Bone,Name="lowerarm_l")
    OldParent=(Type=Bone,Name="lowerarm_l")
    bChildInitial=True
    bNewParentInitial=false
    bOldParentInitial=True
    NewParent.Name=@n121.Item.Name
    OldParent.Name=@n121.Item.Name
    Child.Name=@n119.Item.Name
   )
   n123 = &RigUnit_CollectionItemAtIndex(Index=0, Collection=@n117.Result)
   n124 = &RigUnit_ProjectTransformToNewParent(
    Child=(Type=Bone)
    NewParent=(Type=Bone,Name="lowerarm_l")
    OldParent=(Type=Bone,Name="lowerarm_l")
    bChildInitial=True
    bNewParentInitial=false
    bOldParentInitial=True
    NewParent.Name=@n123.Item.Name
    OldParent.Name=@n123.Item.Name
    Child.Name=@n119.Item.Name
   )
   > &RigUnit_TransformConstraintPerItem(
    BaseItem=(Type=Bone)
    BaseTransform=(Rotation=(X=0,Y=0,Z=0,W=1),Translation=(X=0,Y=0,Z=0),Scale3D=(X=1,Y=1,Z=1))
    BaseTransformSpace=GlobalSpace
    Item=(Type=Bone,Name="elbow_out_volume_base_l")
    Targets="((Transform=(Rotation=(X=0,Y=0,Z=0,W=1),Translation=(X=0,Y=0,Z=0),Scale3D=(X=1,Y=1,Z=1)),Weight=0.25,bMaintainOffset=false,Filter=(TranslationFilter=(bX=True,bY=True,bZ=True),RotationFilter=(bX=True,bY=True,bZ=True),ScaleFilter=(bX=True,bY=True,bZ=True))), ...(truncated)"
    bUseInitialTransforms=True
    Targets.1.Transform=@n120.Transform
    Targets.2.Transform=@n122.Transform
    Targets.0.Transform=@n124.Transform
    Item.Name=@n119.Item.Name
   )
   n125 = &RigUnit_CollectionItems(Items=((Type=Bone,Name="pec_volume_aim_l")))
   n126 = &RigUnit_CollectionReplaceItems(New=_r, Old=_l, Items=@n125.Collection)
   n127 = &RigUnit_CollectionItems(Items=((Type=Bone,Name="upperarm_l")))
   n128 = &RigUnit_CollectionReplaceItems(New=_r, Old=_l, Items=@n127.Collection)
   > &RigUnit_AimItem(
    DebugSettings=(Scale=10,WorldOffset=(Rotation=(X=0,Y=0,Z=0,W=1),Translation=(X=0,Y=0,Z=0),Scale3D=(X=1,Y=1,Z=1)))
    Item=(Type=Bone,Name="pec_volume_aim_l")
    Primary=(Weight=1,Axis=(X=1,Y=0,Z=0),Target=(X=0,Y=0,Z=0),Kind=Location,Space=(Type=Bone,Name="upperarm_l"))
    Secondary=(Weight=0,Axis=(X=0,Y=0,Z=1),Target=(X=1,Y=0,Z=0),Kind=Location,Space=(Type=Bone))
    Weight=1
    Item.Name=&RigUnit_CollectionItemAtIndex(
     Index=0
     Collection=RigVMDispatchNode(Values=((),()), Values.0=@n125.Collection, Values.1=@n126.Collection, Index=@n115.Index) {template="DISPATCH_RigVMDispatch_SelectInt32(in Index,in Values,out Result)"}.Result
    ).Item.Name
    Primary.Space.Name=&RigUnit_CollectionItemAtIndex(
     Index=0
     Collection=RigVMDispatchNode(Values=((),()), Values.0=@n127.Collection, Values.1=@n128.Collection, Index=@n115.Index) {template="DISPATCH_RigVMDispatch_SelectInt32(in Index,in Values,out Result)"}.Result
    ).Item.Name
    Primary.Axis=RigVMDispatchNode(
     Values=((X=0,Y=0,Z=0),(X=0,Y=0,Z=0))
     Values.0=/Script/RigVM.RigVMFunction_MathVectorAdd(A=(X=1,Y=0,Z=0), B=(X=0,Y=0,Z=0)) {is=struct RigVMUnitNode, template="Add::Execute(in A,in B,out Result)"}.Result
     Values.1=&RigVMFunction_MathVectorAdd(A=(X=-1,Y=0,Z=0), B=(X=0,Y=0,Z=0)) {template="Add::Execute(in A,in B,out Result)"}.Result
     Index=@n115.Index
    ) {template="DISPATCH_RigVMDispatch_SelectInt32(in Index,in Values,out Result)"}.Result
   )
   n129 = &RigUnit_SequenceExecution(ExecuteContext=@n33.C)
   n130 = &RigVMFunction_ForLoopCount(Count=2, ExecutePin=@n129.A)
   n131 = &RigUnit_CollectionItems(Items=((Type=Bone,Name="upperarm_twist_01_l"),(Type=Bone,Name="clavicle_l"),(Type=Bone,Name="upperarm_up_volume_base_l")))
   n132 = RigVMDispatchNode(
    Values=((),())
    Values.0=@n131.Collection
    Values.1=&RigUnit_CollectionReplaceItems(New=_r, Old=_l, Items=@n131.Collection).Collection
    Index=@n130.Index
   ) {template="DISPATCH_RigVMDispatch_SelectInt32(in Index,in Values,out Result)"}
   n133 = &RigUnit_CollectionItemAtIndex(Index=1, Collection=@n132.Result)
   n134 = &RigUnit_CollectionItemAtIndex(Index=2, Collection=@n132.Result)
   n135 = &RigUnit_ProjectTransformToNewParent(
    Child=(Type=Bone)
    NewParent=(Type=Bone,Name="lowerarm_l")
    OldParent=(Type=Bone,Name="lowerarm_l")
    bChildInitial=True
    bNewParentInitial=false
    bOldParentInitial=True
    NewParent.Name=@n133.Item.Name
    OldParent.Name=@n133.Item.Name
    Child.Name=@n134.Item.Name
   )
   n136 = &RigUnit_CollectionItemAtIndex(Index=0, Collection=@n132.Result)
   n137 = &RigUnit_ProjectTransformToNewParent(
    Child=(Type=Bone)
    NewParent=(Type=Bone,Name="lowerarm_l")
    OldParent=(Type=Bone,Name="lowerarm_l")
    bChildInitial=True
    bNewParentInitial=false
    bOldParentInitial=True
    NewParent.Name=@n136.Item.Name
    OldParent.Name=@n136.Item.Name
    Child.Name=@n134.Item.Name
   )
   > &RigUnit_TransformConstraintPerItem(
    BaseItem=(Type=Bone)
    BaseTransform=(Rotation=(X=0,Y=0,Z=0,W=1),Translation=(X=0,Y=0,Z=0),Scale3D=(X=1,Y=1,Z=1))
    BaseTransformSpace=GlobalSpace
    Item=(Type=Bone,Name="elbow_out_volume_base_l")
    Targets="((Transform=(Rotation=(X=0,Y=0,Z=0,W=1),Translation=(X=0,Y=0,Z=0),Scale3D=(X=1,Y=1,Z=1)),Weight=0.6,bMaintainOffset=false,Filter=(TranslationFilter=(bX=True,bY=True,bZ=True),RotationFilter=(bX=True,bY=True,bZ=True),ScaleFilter=(bX=True,bY=True,bZ=True))),( ...(truncated)"
    bUseInitialTransforms=True
    Targets.1.Transform=@n135.Transform
    Targets.0.Transform=@n137.Transform
    Item.Name=@n134.Item.Name
   )
   n138 = &RigUnit_CollectionItems(Items=((Type=Bone,Name="upperarm_up_volume_l")))
   n139 = &RigUnit_CollectionReplaceItems(New=_r, Old=_l, Items=@n138.Collection)
   n140 = &RigVMFunction_MathVectorMul(A=(X=0,Y=0,Z=8), B=(X=1,Y=1,Z=1)) {template="Multiply::Execute(in A,in B,out Result)"}
   > &RigUnit_ModifyTransforms(
    ItemToModify=((Item=(Type=Bone,Name="elbow_out_volume_l"),Transform=(Rotation=(X=0,Y=0,Z=0,W=1),Translation=(X=0,Y=-3,Z=0),Scale3D=(X=1,Y=1,Z=1))))
    ItemToModify.0.Item.Name=&RigUnit_CollectionItemAtIndex(
     Index=0
     Collection=RigVMDispatchNode(Values=((),()), Values.0=@n138.Collection, Values.1=@n139.Collection, Index=@n130.Index) {template="DISPATCH_RigVMDispatch_SelectInt32(in Index,in Values,out Result)"}.Result
    ).Item.Name
    Weight=RigVMDispatchNode(
     Values=(0,0)
     Values.0=RigVMVariableNode() {variable=psd_arm_up_alpha_l}.Value
     Values.1=RigVMVariableNode() {variable=psd_arm_up_alpha_r}.Value
     Index=@n130.Index
    ) {template="DISPATCH_RigVMDispatch_SelectInt32(in Index,in Values,out Result)"}.Result
    ItemToModify.0.Transform.Translation=RigVMDispatchNode(
     Values=((X=0,Y=8,Z=0),(X=0,Y=-8,Z=0))
     Values.0=@n140.Result
     Values.1=&RigVMFunction_MathVectorMul(B=(X=-1,Y=-1,Z=-1), A=@n140.Result) {template="Multiply::Execute(in A,in B,out Result)"}.Result
     Index=@n130.Index
    ) {template="DISPATCH_RigVMDispatch_SelectInt32(in Index,in Values,out Result)"}.Result
   )
   n141 = &RigVMFunction_ForLoopCount(Count=2, ExecutePin=@n129.B)
   n142 = &RigUnit_CollectionItems(Items=((Type=Bone,Name="upperarm_twist_01_l"),(Type=Bone,Name="clavicle_l"),(Type=Bone,Name="upperarm_fr_volume_base_l")))
   n143 = RigVMDispatchNode(
    Values=((),())
    Values.0=@n142.Collection
    Values.1=&RigUnit_CollectionReplaceItems(New=_r, Old=_l, Items=@n142.Collection).Collection
    Index=@n141.Index
   ) {template="DISPATCH_RigVMDispatch_SelectInt32(in Index,in Values,out Result)"}
   n144 = &RigUnit_CollectionItemAtIndex(Index=1, Collection=@n143.Result)
   n145 = &RigUnit_CollectionItemAtIndex(Index=2, Collection=@n143.Result)
   n146 = &RigUnit_ProjectTransformToNewParent(
    Child=(Type=Bone)
    NewParent=(Type=Bone,Name="lowerarm_l")
    OldParent=(Type=Bone,Name="lowerarm_l")
    bChildInitial=True
    bNewParentInitial=false
    bOldParentInitial=True
    NewParent.Name=@n144.Item.Name
    OldParent.Name=@n144.Item.Name
    Child.Name=@n145.Item.Name
   )
   n147 = &RigUnit_CollectionItemAtIndex(Index=0, Collection=@n143.Result)
   n148 = &RigUnit_ProjectTransformToNewParent(
    Child=(Type=Bone)
    NewParent=(Type=Bone,Name="lowerarm_l")
    OldParent=(Type=Bone,Name="lowerarm_l")
    bChildInitial=True
    bNewParentInitial=false
    bOldParentInitial=True
    Child.Name=@n145.Item.Name
    NewParent.Name=@n147.Item.Name
    OldParent.Name=@n147.Item.Name
   )
   > &RigUnit_TransformConstraintPerItem(
    BaseItem=(Type=Bone)
    BaseTransform=(Rotation=(X=0,Y=0,Z=0,W=1),Translation=(X=0,Y=0,Z=0),Scale3D=(X=1,Y=1,Z=1))
    BaseTransformSpace=GlobalSpace
    Item=(Type=Bone,Name="elbow_out_volume_base_l")
    Targets="((Transform=(Rotation=(X=0,Y=0,Z=0,W=1),Translation=(X=0,Y=0,Z=0),Scale3D=(X=1,Y=1,Z=1)),Weight=0.3,bMaintainOffset=false,Filter=(TranslationFilter=(bX=True,bY=True,bZ=True),RotationFilter=(bX=True,bY=True,bZ=True),ScaleFilter=(bX=True,bY=True,bZ=True))),( ...(truncated)"
    bUseInitialTransforms=True
    Targets.1.Transform=@n146.Transform
    Targets.0.Transform=@n148.Transform
    Item.Name=@n145.Item.Name
   )
   n149 = &RigUnit_CollectionItems(Items=((Type=Bone,Name="upperarm_fr_volume_l")))
   n150 = &RigUnit_CollectionReplaceItems(New=_r, Old=_l, Items=@n149.Collection)
   n151 = RigVMDispatchNode(
    Values=(0,0)
    Values.1=RigVMVariableNode() {variable=psd_arm_fwd_alpha_r}.Value
    Values.0=RigVMVariableNode() {variable=psd_arm_fwd_alpha_l}.Value
    Index=@n141.Index
   ) {template="DISPATCH_RigVMDispatch_SelectInt32(in Index,in Values,out Result)"}
   n152 = &RigVMFunction_MathVectorMul(A=(X=0,Y=10,Z=0), B=(X=1,Y=1,Z=1)) {template="Multiply::Execute(in A,in B,out Result)"}
   > &RigUnit_ModifyTransforms(
    ItemToModify=((Item=(Type=Bone,Name="elbow_out_volume_l"),Transform=(Rotation=(X=0,Y=0,Z=0,W=1),Translation=(X=0,Y=-3,Z=0),Scale3D=(X=1,Y=1,Z=1))))
    ItemToModify.0.Item.Name=&RigUnit_CollectionItemAtIndex(
     Index=0
     Collection=RigVMDispatchNode(Values=((),()), Values.0=@n149.Collection, Values.1=@n150.Collection, Index=@n141.Index) {template="DISPATCH_RigVMDispatch_SelectInt32(in Index,in Values,out Result)"}.Result
    ).Item.Name
    Weight=@n151.Result
    ItemToModify.0.Transform.Translation=RigVMDispatchNode(
     Values=((X=0,Y=8,Z=0),(X=0,Y=-8,Z=0))
     Values.0=@n152.Result
     Values.1=&RigVMFunction_MathVectorMul(B=(X=-1,Y=-1,Z=-1), A=@n152.Result) {template="Multiply::Execute(in A,in B,out Result)"}.Result
     Index=@n141.Index
    ) {template="DISPATCH_RigVMDispatch_SelectInt32(in Index,in Values,out Result)"}.Result
   )
   n153 = &RigUnit_CollectionItems(Items=((Type=Bone,Name="upperarm_bk_volume_l")))
   n154 = &RigUnit_CollectionReplaceItems(New=_r, Old=_l, Items=@n153.Collection)
   n155 = &RigVMFunction_MathVectorMul(A=(X=0,Y=-3,Z=0), B=(X=1,Y=1,Z=1)) {template="Multiply::Execute(in A,in B,out Result)"}
   > &RigUnit_ModifyTransforms(
    ItemToModify=((Item=(Type=Bone,Name="elbow_out_volume_l"),Transform=(Rotation=(X=0,Y=0,Z=0,W=1),Translation=(X=0,Y=-3,Z=0),Scale3D=(X=1,Y=1,Z=1))))
    ItemToModify.0.Item.Name=&RigUnit_CollectionItemAtIndex(
     Index=0
     Collection=RigVMDispatchNode(Values=((),()), Values.0=@n153.Collection, Values.1=@n154.Collection, Index=@n141.Index) {template="DISPATCH_RigVMDispatch_SelectInt32(in Index,in Values,out Result)"}.Result
    ).Item.Name
    ItemToModify.0.Transform.Translation=RigVMDispatchNode(
     Values=((X=0,Y=8,Z=0),(X=0,Y=-8,Z=0))
     Values.0=@n155.Result
     Values.1=&RigVMFunction_MathVectorMul(B=(X=-1,Y=-1,Z=-1), A=@n155.Result) {template="Multiply::Execute(in A,in B,out Result)"}.Result
     Index=@n141.Index
    ) {template="DISPATCH_RigVMDispatch_SelectInt32(in Index,in Values,out Result)"}.Result
    Weight=@n151.Result
   )
   n156 = &RigUnit_CollectionItems(Items=((Type=Bone,Name="upperarm_fr_volume_l")))
   n157 = &RigUnit_CollectionReplaceItems(New=_r, Old=_l, Items=@n156.Collection)
   n158 = RigVMDispatchNode(
    Values=(0,0)
    Values.0=RigVMVariableNode() {variable=psd_arm_fwd_up_alpha_l}.Value
    Values.1=RigVMVariableNode() {variable=psd_arm_fwd_up_alpha_r}.Value
    Index=@n141.Index
   ) {template="DISPATCH_RigVMDispatch_SelectInt32(in Index,in Values,out Result)"}
   n159 = &RigVMFunction_MathVectorMul(A=(X=0,Y=10,Z=0), B=(X=1,Y=1,Z=1)) {template="Multiply::Execute(in A,in B,out Result)"}
   > &RigUnit_ModifyTransforms(
    ItemToModify=((Item=(Type=Bone,Name="elbow_out_volume_l"),Transform=(Rotation=(X=0,Y=0,Z=0,W=1),Translation=(X=0,Y=-3,Z=0),Scale3D=(X=0.7,Y=1,Z=1))))
    ItemToModify.0.Item.Name=&RigUnit_CollectionItemAtIndex(
     Index=0
     Collection=RigVMDispatchNode(Values=((),()), Values.0=@n156.Collection, Values.1=@n157.Collection, Index=@n141.Index) {template="DISPATCH_RigVMDispatch_SelectInt32(in Index,in Values,out Result)"}.Result
    ).Item.Name
    Weight=@n158.Result
    ItemToModify.0.Transform.Translation=RigVMDispatchNode(
     Values=((X=0,Y=8,Z=0),(X=0,Y=-8,Z=0))
     Values.0=@n159.Result
     Values.1=&RigVMFunction_MathVectorMul(B=(X=-1,Y=-1,Z=-1), A=@n159.Result) {template="Multiply::Execute(in A,in B,out Result)"}.Result
     Index=@n141.Index
    ) {template="DISPATCH_RigVMDispatch_SelectInt32(in Index,in Values,out Result)"}.Result
   )
   n160 = &RigUnit_CollectionItems(Items=((Type=Bone,Name="upperarm_up_volume_l")))
   n161 = &RigUnit_CollectionReplaceItems(New=_r, Old=_l, Items=@n160.Collection)
   n162 = &RigVMFunction_MathVectorMul(A=(X=0,Y=5,Z=0), B=(X=1,Y=1,Z=1)) {template="Multiply::Execute(in A,in B,out Result)"}
   > &RigUnit_ModifyTransforms(
    ItemToModify=((Item=(Type=Bone,Name="elbow_out_volume_l"),Transform=(Rotation=(X=0,Y=0,Z=0,W=1),Translation=(X=0,Y=-3,Z=0),Scale3D=(X=1,Y=1,Z=1))))
    ItemToModify.0.Item.Name=&RigUnit_CollectionItemAtIndex(
     Index=0
     Collection=RigVMDispatchNode(Values=((),()), Values.0=@n160.Collection, Values.1=@n161.Collection, Index=@n141.Index) {template="DISPATCH_RigVMDispatch_SelectInt32(in Index,in Values,out Result)"}.Result
    ).Item.Name
    ItemToModify.0.Transform.Translation=RigVMDispatchNode(
     Values=((X=0,Y=8,Z=0),(X=0,Y=-8,Z=0))
     Values.0=@n162.Result
     Values.1=&RigVMFunction_MathVectorMul(B=(X=-1,Y=-1,Z=-1), A=@n162.Result) {template="Multiply::Execute(in A,in B,out Result)"}.Result
     Index=@n141.Index
    ) {template="DISPATCH_RigVMDispatch_SelectInt32(in Index,in Values,out Result)"}.Result
    Weight=@n158.Result
   )
   n163 = &RigVMFunction_ForLoopCount(Count=2, ExecutePin=@n129.C)
   n164 = &RigUnit_CollectionItems(Items=((Type=Bone,Name="upperarm_twist_01_l"),(Type=Bone,Name="clavicle_l"),(Type=Bone,Name="upperarm_bk_volume_base_l")))
   n165 = RigVMDispatchNode(
    Values=((),())
    Values.0=@n164.Collection
    Values.1=&RigUnit_CollectionReplaceItems(New=_r, Old=_l, Items=@n164.Collection).Collection
    Index=@n163.Index
   ) {template="DISPATCH_RigVMDispatch_SelectInt32(in Index,in Values,out Result)"}
   n166 = &RigUnit_CollectionItemAtIndex(Index=1, Collection=@n165.Result)
   n167 = &RigUnit_CollectionItemAtIndex(Index=2, Collection=@n165.Result)
   n168 = &RigUnit_ProjectTransformToNewParent(
    Child=(Type=Bone)
    NewParent=(Type=Bone,Name="lowerarm_l")
    OldParent=(Type=Bone,Name="lowerarm_l")
    bChildInitial=True
    bNewParentInitial=false
    bOldParentInitial=True
    NewParent.Name=@n166.Item.Name
    OldParent.Name=@n166.Item.Name
    Child.Name=@n167.Item.Name
   )
   n169 = &RigUnit_CollectionItemAtIndex(Index=0, Collection=@n165.Result)
   n170 = &RigUnit_ProjectTransformToNewParent(
    Child=(Type=Bone)
    NewParent=(Type=Bone,Name="lowerarm_l")
    OldParent=(Type=Bone,Name="lowerarm_l")
    bChildInitial=True
    bNewParentInitial=false
    bOldParentInitial=True
    Child.Name=@n167.Item.Name
    NewParent.Name=@n169.Item.Name
    OldParent.Name=@n169.Item.Name
   )
   > &RigUnit_TransformConstraintPerItem(
    BaseItem=(Type=Bone)
    BaseTransform=(Rotation=(X=0,Y=0,Z=0,W=1),Translation=(X=0,Y=0,Z=0),Scale3D=(X=1,Y=1,Z=1))
    BaseTransformSpace=GlobalSpace
    Item=(Type=Bone,Name="elbow_out_volume_base_l")
    Targets="((Transform=(Rotation=(X=0,Y=0,Z=0,W=1),Translation=(X=0,Y=0,Z=0),Scale3D=(X=1,Y=1,Z=1)),Weight=0.2,bMaintainOffset=false,Filter=(TranslationFilter=(bX=True,bY=True,bZ=True),RotationFilter=(bX=True,bY=True,bZ=True),ScaleFilter=(bX=True,bY=True,bZ=True))),( ...(truncated)"
    bUseInitialTransforms=True
    Targets.1.Transform=@n168.Transform
    Targets.0.Transform=@n170.Transform
    Item.Name=@n167.Item.Name
   )
   n171 = &RigUnit_CollectionItems(Items=((Type=Bone,Name="upperarm_bk_volume_l")))
   n172 = &RigUnit_CollectionReplaceItems(New=_r, Old=_l, Items=@n171.Collection)
   n173 = &RigVMFunction_MathVectorMul(A=(X=0,Y=-4,Z=0), B=(X=1,Y=1,Z=1)) {template="Multiply::Execute(in A,in B,out Result)"}
   > &RigUnit_ModifyTransforms(
    ItemToModify=((Item=(Type=Bone,Name="elbow_out_volume_l"),Transform=(Rotation=(X=0,Y=0,Z=0,W=1),Translation=(X=0,Y=-3,Z=0),Scale3D=(X=1,Y=1,Z=1))))
    ItemToModify.0.Item.Name=&RigUnit_CollectionItemAtIndex(
     Index=0
     Collection=RigVMDispatchNode(Values=((),()), Values.0=@n171.Collection, Values.1=@n172.Collection, Index=@n163.Index) {template="DISPATCH_RigVMDispatch_SelectInt32(in Index,in Values,out Result)"}.Result
    ).Item.Name
    Weight=RigVMDispatchNode(
     Values=(0,0)
     Values.0=RigVMVariableNode() {variable=psd_arm_bk_alpha_l}.Value
     Values.1=RigVMVariableNode() {variable=psd_arm_bk_alpha_r}.Value
     Index=@n163.Index
    ) {template="DISPATCH_RigVMDispatch_SelectInt32(in Index,in Values,out Result)"}.Result
    ItemToModify.0.Transform.Translation=RigVMDispatchNode(
     Values=((X=0,Y=8,Z=0),(X=0,Y=-8,Z=0))
     Values.0=@n173.Result
     Values.1=&RigVMFunction_MathVectorMul(B=(X=-1,Y=-1,Z=-1), A=@n173.Result) {template="Multiply::Execute(in A,in B,out Result)"}.Result
     Index=@n163.Index
    ) {template="DISPATCH_RigVMDispatch_SelectInt32(in Index,in Values,out Result)"}.Result
   )
   &RigUnit_BeginExecution()
   &RigUnit_BeginExecution()
   n174 = &RigUnit_CollectionItems(Items=((Type=Bone,Name="pec_volume_aim_l"),(Type=Bone,Name="upperarm_l")))
   n175 = RigVMDispatchNode(
    Values=((),())
    Values.0=@n174.Collection
    Values.1=&RigUnit_CollectionReplaceItems(New=_r, Old=_l, Items=@n174.Collection).Collection
    Index=@n115.Index
   ) {template="DISPATCH_RigVMDispatch_SelectInt32(in Index,in Values,out Result)"}
   n176 = &RigUnit_CollectionItemAtIndex(Index=1, Collection=@n175.Result)
   n177 = &RigUnit_CollectionItems(Items=((Type=Bone,Name="pec_volume_l")))
   n178 = &RigUnit_CollectionItemAtIndex(
    Index=0
    Collection=RigVMDispatchNode(
     Values=((),())
     Values.0=@n177.Collection
     Values.1=&RigUnit_CollectionReplaceItems(New=_r, Old=_l, Items=@n177.Collection).Collection
     Index=@n115.Index
    ) {template="DISPATCH_RigVMDispatch_SelectInt32(in Index,in Values,out Result)"}.Result
   )
   n179 = &RigUnit_ProjectTransformToNewParent(
    Child=(Type=Bone)
    NewParent=(Type=Bone,Name="lowerarm_l")
    OldParent=(Type=Bone,Name="lowerarm_l")
    bChildInitial=True
    bNewParentInitial=false
    bOldParentInitial=True
    NewParent.Name=@n176.Item.Name
    OldParent.Name=@n176.Item.Name
    Child.Name=@n178.Item.Name
   )
   n180 = &RigUnit_CollectionItemAtIndex(Index=0, Collection=@n175.Result)
   n181 = &RigUnit_ProjectTransformToNewParent(
    Child=(Type=Bone)
    NewParent=(Type=Bone,Name="lowerarm_l")
    OldParent=(Type=Bone,Name="lowerarm_l")
    bChildInitial=True
    bNewParentInitial=false
    bOldParentInitial=True
    NewParent.Name=@n180.Item.Name
    OldParent.Name=@n180.Item.Name
    Child.Name=@n178.Item.Name
   )
   &RigUnit_TransformConstraintPerItem(
    BaseItem=(Type=Bone)
    BaseTransform=(Rotation=(X=0,Y=0,Z=0,W=1),Translation=(X=0,Y=0,Z=0),Scale3D=(X=1,Y=1,Z=1))
    BaseTransformSpace=GlobalSpace
    ExecutePin=()
    Item=(Type=Bone,Name="elbow_out_volume_base_l")
    Targets="((Transform=(Rotation=(X=0,Y=0,Z=0,W=1),Translation=(X=0,Y=0,Z=0),Scale3D=(X=1,Y=1,Z=1)),Weight=0.5,bMaintainOffset=false,Filter=(TranslationFilter=(bX=True,bY=True,bZ=True),RotationFilter=(bX=True,bY=True,bZ=True),ScaleFilter=(bX=True,bY=True,bZ=True))),( ...(truncated)"
    bUseInitialTransforms=True
    Targets.1.Transform=@n179.Transform
    Targets.0.Transform=@n181.Transform
    Item.Name=@n178.Item.Name
   )
  ]}
]
hierarchy [
 root {loc=<0 0 0>, rot=<0 0 0 1>}
 pelvis {parent=root, loc=<0 -0.5301 98.0571>, rot=<0.0127 0.707 -0.0127 -0.707>}
 spine_01 {parent=pelvis, loc=<6.1208 -0.9328 0>, rot=<0 0 0.034 -0.9994>}
 spine_02 {parent=spine_01, loc=<6.5996 -0.2095 0>, rot=<0 0 0.0238 -0.9997>}
 spine_03 {parent=spine_02, loc=<6.4914 -0.1237 0>, rot=<0 0 0.0149 -0.9999>}
 spine_04 {parent=spine_03, loc=<6.9598 0.0794 0>, rot=<0 0 -0.0132 -0.9999>}
 spine_05 {parent=spine_04, loc=<6.9987 0 0>, rot=<0 0 -0.0015 -1>}
 clavicle_l {parent=spine_05, loc=<5.4831 -0.1053 -3.2946>, rot=<0.0465 -0.7289 -0.009 -0.6829>}
 upperarm_l {parent=clavicle_l, loc=<9.8658 -1.2136 -0.4105>, rot=<0.0164 -0.3245 -0.0269 -0.9454>}
 lowerarm_l {parent=upperarm_l, loc=<26.465 -0.8345 0.259>, rot=<0.0036 0.0161 -0.2298 -0.9731>}
 hand_l {parent=lowerarm_l, loc=<22.7355 -0.3203 -0.6105>, rot=<-0.6784 -0.0223 0.0735 -0.7307>}
 index_metacarpal_l {parent=hand_l, loc=<1.6535 0.0183 -2.96>, rot=<0.0158 -0.0124 -0.0173 -0.9996>}
 index_01_l {parent=index_metacarpal_l, loc=<8.3008 0.0357 -0.0524>, rot=<-0.0044 0.0485 0.0924 -0.9945>}
 index_02_l {parent=index_01_l, loc=<3.3587 -0.3004 0.0452>, rot=<-0.002 0.0027 0.3225 -0.9466>}
 index_03_l {parent=index_02_l, loc=<2.5578 -0.123 -0.0252>, rot=<-0.002 0.0067 0.2558 -0.9667>}
 middle_metacarpal_l {parent=hand_l, loc=<2.1856 -0.1725 -0.5984>, rot=<0.006 0.0168 -0.0329 -0.9993>}
 middle_01_l {parent=middle_metacarpal_l, loc=<7.3275 -0.2447 0.0129>, rot=<-0.0208 0.0042 0.0858 -0.9961>}
 middle_02_l {parent=middle_01_l, loc=<4.0123 0.1401 0.1175>, rot=<-0.0218 0.0258 0.3512 -0.9357>}
 middle_03_l {parent=middle_02_l, loc=<2.6077 0.0534 0.1373>, rot=<0.0082 0.0188 0.1708 -0.9851>}
 pinky_metacarpal_l {parent=hand_l, loc=<1.7616 -0.9753 3.3814>, rot=<-0.0397 0.0814 -0.069 -0.9935>}
 pinky_01_l {parent=pinky_metacarpal_l, loc=<7.1709 -0.4831 -0.1094>, rot=<-0.0654 -0.0035 0.2242 -0.9723>}
 pinky_02_l {parent=pinky_01_l, loc=<2.8274 0.0893 -0.1043>, rot=<0.0037 0.0011 0.2854 -0.9584>}
 pinky_03_l {parent=pinky_02_l, loc=<2.2046 0.0307 -0.0078>, rot=<-0.0002 0.0098 0.1445 -0.9895>}
 ring_metacarpal_l {parent=hand_l, loc=<2.2074 -0.3625 1.3638>, rot=<-0.0284 0.0722 -0.0458 -0.9959>}
 ring_01_l {parent=ring_metacarpal_l, loc=<7.0295 -0.6197 -0.1446>, rot=<-0.0489 -0.0252 0.1892 -0.9804>}
 ring_02_l {parent=ring_01_l, loc=<3.6174 0.4647 -0.1027>, rot=<0 0 0.2806 -0.9598>}
 ring_03_l {parent=ring_02_l, loc=<2.4152 0.114 0.0199>, rot=<0 0 0.1189 -0.9929>}
 thumb_01_l {parent=hand_l, loc=<1.3118 -0.3191 -2.459>, rot=<0.6546 -0.4042 0.0664 -0.6354>}
 thumb_02_l {parent=thumb_01_l, loc=<5.9336 -0.0448 0.0401>, rot=<-0.0037 0.0213 0.298 -0.9543>}
 thumb_03_l {parent=thumb_02_l, loc=<2.2223 0.3682 -0.0907>, rot=<0.0118 -0.013 0.0066 -0.9998>}
 lowerarm_twist_01_l {parent=lowerarm_l, loc=<18.316 0 0>, rot=<0 0 0 -1>}
 lowerarm_twist_02_l {parent=lowerarm_l, loc=<10.6707 0 0>, rot=<0 0 0 -1>}
 elbow_out_volume_base_l {parent=lowerarm_l, loc=<0 -4 0>, rot=<0 0 0 -1>}
 elbow_out_volume_l {parent=elbow_out_volume_base_l, loc=<0 0 0>, rot=<0 0 0 -1>}
 elbow_in_volume_base_l {parent=lowerarm_l, loc=<0.1834 3.6062 -0.2572>, rot=<0 0 0 -1>}
 elbow_in_volume_l {parent=elbow_in_volume_base_l, loc=<0 0 0>, rot=<0 0 0 -1>}
 upperarm_twist_01_l {parent=upperarm_l, loc=<7.2135 -0.1134 0>, rot=<0 0 0 1>}
 upperarm_bk_volume_base_l {parent=upperarm_twist_01_l, loc=<-6.4053 -4.3746 0.314>, rot=<0 0 0 1>}
 upperarm_bk_volume_l {parent=upperarm_bk_volume_base_l, loc=<0 0 0>, rot=<0 0 0 1>}
 upperarm_fr_volume_base_l {parent=upperarm_twist_01_l, loc=<-5.6312 5.092 0.3354>, rot=<0 0 0 1>}
 upperarm_fr_volume_l {parent=upperarm_fr_volume_base_l, loc=<0 0 0>, rot=<0 0 0 1>}
 upperarm_up_volume_base_l {parent=upperarm_twist_01_l, loc=<-8.891 0.4825 3.9517>, rot=<0 0 0 1>}
 upperarm_up_volume_l {parent=upperarm_up_volume_base_l, loc=<0 0 0>, rot=<0 0 0 1>}
 upperarm_twist_02_l {parent=upperarm_l, loc=<19.7998 -0.1134 0>, rot=<0 0 0 1>}
 pose_arm_up_l {parent=clavicle_l, loc=<27.0599 -1.9836 19.7124>, rot=<0.0008 0.0336 -0.0252 -0.9991>}
 pose_arm_fwd_l {parent=clavicle_l, loc=<6.0547 24.9809 -1.1065>, rot=<0.0008 0.0336 -0.0252 -0.9991>}
 pose_arm_bk_l {parent=clavicle_l, loc=<21.3409 -25.0308 1.0813>, rot=<0.0008 0.0336 -0.0252 -0.9991>}
 pose_arm_fwd_dn_l {parent=clavicle_l, loc=<10.1535 15.7305 -20.7561>, rot=<0.0008 0.0336 -0.0252 -0.9991>}
 pose_arm_fwd_up_l {parent=clavicle_l, loc=<9.0897 20.4292 14.8259>, rot=<0.0008 0.0336 -0.0252 -0.9991>}
 clavicle_r {parent=spine_05, loc=<5.4831 -0.1053 3.2946>, rot=<0.7289 0.0465 -0.6829 0.009>}
 upperarm_r {parent=clavicle_r, loc=<-9.8658 1.2136 0.4104>, rot=<0.0164 -0.3245 -0.0269 -0.9454>}
 lowerarm_r {parent=upperarm_r, loc=<-26.4651 0.8345 -0.259>, rot=<0.0036 0.0161 -0.2298 -0.9731>}
 hand_r {parent=lowerarm_r, loc=<-22.7356 0.3203 0.6106>, rot=<-0.6784 -0.0223 0.0735 -0.7307>}
 index_metacarpal_r {parent=hand_r, loc=<-1.6534 -0.0181 2.9601>, rot=<0.0158 -0.0124 -0.0173 -0.9996>}
 index_01_r {parent=index_metacarpal_r, loc=<-8.3008 -0.0357 0.0524>, rot=<-0.0044 0.0485 0.0924 -0.9945>}
 index_02_r {parent=index_01_r, loc=<-3.3589 0.3003 -0.0453>, rot=<-0.002 0.0027 0.3225 -0.9466>}
 index_03_r {parent=index_02_r, loc=<-2.5578 0.123 0.0252>, rot=<-0.002 0.0067 0.2558 -0.9667>}
 middle_metacarpal_r {parent=hand_r, loc=<-2.1856 0.1726 0.5984>, rot=<0.006 0.0168 -0.0329 -0.9993>}
 middle_01_r {parent=middle_metacarpal_r, loc=<-7.3273 0.2447 -0.0129>, rot=<-0.0208 0.0042 0.0858 -0.9961>}
 middle_02_r {parent=middle_01_r, loc=<-4.0124 -0.14 -0.1175>, rot=<-0.0218 0.0258 0.3512 -0.9357>}
 middle_03_r {parent=middle_02_r, loc=<-2.6077 -0.0533 -0.1373>, rot=<0.0082 0.0188 0.1708 -0.9851>}
 pinky_metacarpal_r {parent=hand_r, loc=<-1.7617 0.9753 -3.3814>, rot=<-0.0397 0.0814 -0.069 -0.9935>}
 pinky_01_r {parent=pinky_metacarpal_r, loc=<-7.1707 0.4832 0.1094>, rot=<-0.0654 -0.0035 0.2242 -0.9723>}
 pinky_02_r {parent=pinky_01_r, loc=<-2.8274 -0.0892 0.1043>, rot=<0.0037 0.0011 0.2854 -0.9584>}
 pinky_03_r {parent=pinky_02_r, loc=<-2.2045 -0.031 0.0078>, rot=<-0.0002 0.0098 0.1445 -0.9895>}
 ring_metacarpal_r {parent=hand_r, loc=<-2.2073 0.3627 -1.3638>, rot=<-0.0284 0.0722 -0.0458 -0.9959>}
 ring_01_r {parent=ring_metacarpal_r, loc=<-7.0297 0.6197 0.1446>, rot=<-0.0489 -0.0252 0.1892 -0.9804>}
 ring_02_r {parent=ring_01_r, loc=<-3.6171 -0.4648 0.1027>, rot=<0 0 0.2806 -0.9598>}
 ring_03_r {parent=ring_02_r, loc=<-2.4153 -0.1141 -0.0199>, rot=<0 0 0.1189 -0.9929>}
 thumb_01_r {parent=hand_r, loc=<-1.3119 0.3192 2.459>, rot=<0.6546 -0.4042 0.0664 -0.6354>}
 thumb_02_r {parent=thumb_01_r, loc=<-5.9336 0.0447 -0.0401>, rot=<-0.0037 0.0213 0.298 -0.9543>}
 thumb_03_r {parent=thumb_02_r, loc=<-2.2223 -0.3682 0.0907>, rot=<0.0118 -0.013 0.0066 -0.9998>}
 lowerarm_twist_01_r {parent=lowerarm_r, loc=<-18.3161 0 0>, rot=<0 0 0 1>}
 lowerarm_twist_02_r {parent=lowerarm_r, loc=<-10.6707 0 0>, rot=<0 0 0 1>}
 elbow_out_volume_base_r {parent=lowerarm_r, loc=<0 4 0>, rot=<0 0 0 1>}
 elbow_out_volume_r {parent=elbow_out_volume_base_r, loc=<0 0 0>, rot=<0 0 0 1>}
 elbow_in_volume_base_r {parent=lowerarm_r, loc=<-0.1834 -3.6062 0.2572>, rot=<0 0 0 -1>}
 elbow_in_volume_r {parent=elbow_in_volume_base_r, loc=<0 0 0>, rot=<0 0 0 1>}
 upperarm_twist_01_r {parent=upperarm_r, loc=<-7.2135 0.1134 0>, rot=<0 0 0 1>}
 upperarm_bk_volume_base_r {parent=upperarm_twist_01_r, loc=<6.4052 4.3746 -0.3139>, rot=<0 0 0 1>}
 upperarm_bk_volume_r {parent=upperarm_bk_volume_base_r, loc=<0 0 0>, rot=<0 0 0 1>}
 upperarm_fr_volume_base_r {parent=upperarm_twist_01_r, loc=<5.6311 -5.092 -0.3353>, rot=<0 0 0 1>}
 upperarm_fr_volume_r {parent=upperarm_fr_volume_base_r, loc=<0 0 0>, rot=<0 0 0 1>}
 upperarm_up_volume_base_r {parent=upperarm_twist_01_r, loc=<8.8909 -0.4825 -3.9516>, rot=<0 0 0 1>}
 upperarm_up_volume_r {parent=upperarm_up_volume_base_r, loc=<0 0 0>, rot=<0 0 0 1>}
 upperarm_twist_02_r {parent=upperarm_r, loc=<-19.7998 0.1134 0>, rot=<0 0 0 1>}
 pose_arm_fwd_r {parent=clavicle_r, loc=<-6.0546 -24.981 1.1065>, rot=<0.9991 0.0252 0.0336 0.0008>}
 pose_arm_bk_r {parent=clavicle_r, loc=<-21.3409 25.0309 -1.0813>, rot=<0.9991 0.0252 0.0336 0.0008>}
 pose_arm_up_r {parent=clavicle_r, loc=<-27.0599 1.9836 -19.7125>, rot=<0.9991 0.0252 0.0336 0.0008>}
 pose_arm_fwd_up_r {parent=clavicle_r, loc=<-9.0895 -20.4296 -14.8255>, rot=<0.9991 0.0252 0.0336 0.0008>}
 pose_arm_fwd_dn_r {parent=clavicle_r, loc=<-10.1533 -15.7309 20.7566>, rot=<0.9991 0.0252 0.0336 0.0008>}
 neck_01 {parent=spine_05, loc=<11.3138 0.1529 0>, rot=<0 0 -0.2003 -0.9797>}
 neck_02 {parent=neck_01, loc=<4.5627 0 0>, rot=<0 0 0.0504 -0.9987>}
 head {parent=neck_02, loc=<5.3914 0 0>, rot=<0 0 0.115 -0.9934>}
 pec_volume_aim_base_l {parent=spine_04, loc=<3.1554 10.4127 -1.2237>, rot=<0.2874 -0.4733 0.2345 -0.799>}
 pec_volume_aim_l {parent=pec_volume_aim_base_l, loc=<0 0 0>, rot=<0 0 0 1>}
 pec_volume_base_l {parent=pec_volume_aim_l, loc=<9.3665 2.7609 -1.2773>, rot=<-0.1707 -0.1638 -0.1203 -0.9641>}
 pec_volume_l {parent=pec_volume_base_l, loc=<0 0 0>, rot=<0 0 0 1>}
 pec_volume_aim_base_r {parent=spine_04, loc=<3.155 10.4127 1.2237>, rot=<-0.4733 -0.2874 0.799 0.2345>}
 pec_volume_aim_r {parent=pec_volume_aim_base_r, loc=<0 0 0>, rot=<0 0 0 -1>}
 pec_volume_base_r {parent=pec_volume_aim_r, loc=<-9.3667 -2.7609 1.2771>, rot=<-0.1707 -0.1638 -0.1203 -0.9641>}
 pec_volume_r {parent=pec_volume_base_r, loc=<0 0 0>, rot=<0 0 0 -1>}
 thigh_l {parent=pelvis, loc=<-1.0855 -0.2109 -8.7053>, rot=<0.0009 0.0355 0.0155 -0.9993>}
 calf_l {parent=thigh_l, loc=<-42.5284 -0.0019 -0.7297>, rot=<0.0006 -0.0161 -0.0391 -0.9991>}
 calf_twist_01_l {parent=calf_l, loc=<-35.8483 -1.9197 -0.5794>, rot=<0.0237 0.0095 -0.0035 -0.9997>}
 calf_twist_02_l {parent=calf_l, loc=<-19.1546 -1.7943 -0.2638>, rot=<0.0237 0.0095 -0.0035 -0.9997>}
 foot_l {parent=calf_l, loc=<-46.7401 -2.5014 -0.592>, rot=<0.0427 -0.0238 0.0353 -0.9982>}
 ball_l {parent=foot_l, loc=<-3.5658 10.8066 -0.1544>, rot=<0.0001 -0.0001 -0.7063 -0.7079>}
 knee_out_volume_base_l {parent=calf_l, loc=<0 4 0>, rot=<0 0 0 -1>}
 knee_out_volume_l {parent=knee_out_volume_base_l, loc=<0 0 0>, rot=<0 0 0 -1>}
 knee_in_volume_base_l {parent=calf_l, loc=<0.8031 -5.0231 0.2456>, rot=<0 0 0 -1>}
 knee_in_volume_l {parent=knee_in_volume_base_l, loc=<0 0 0>, rot=<0 0 0 -1>}
 dyn_ankle_l {parent=calf_l, loc=<-38.2914 -2.8551 -0.9919>, rot=<0.032 -0.0235 0.0356 -0.9986>}
 dyn_ball_l {parent=dyn_ankle_l, loc=<-12.0273 10.5393 -0.3817>, rot=<0.0077 -0.0077 -0.7062 -0.7079>}
 thigh_twist_01_l {parent=thigh_l, loc=<-27.952 0 0>, rot=<0.0013 0 0 -1>}
 glute_volume_base_l {parent=thigh_l, loc=<-6 -9 0>, rot=<0 0 -0.4617 -0.887>}
 glute_volume_l {parent=glute_volume_base_l, loc=<0 0 0>, rot=<0 0 0 -1>}
 crouch_volume_base_l {parent=thigh_l, loc=<0.4801 5.6776 -0.2486>, rot=<0 0 0 -1>}
 crouch_volume_l {parent=crouch_volume_base_l, loc=<0 0 0>, rot=<0 0 0 -1>}
 dyn_gun_base {parent=thigh_l, loc=<-16.6037 0.8629 -7.0866>, rot=<-0.0024 -0.9994 -0.0003 0.0355>}
 dyn_gun {parent=dyn_gun_base, loc=<0 0 0>, rot=<0 0 0 -1>}
 thigh_r {parent=pelvis, loc=<-1.0853 -0.2109 8.7053>, rot=<0.0355 -0.0009 0.9993 0.0155>}
 calf_r {parent=thigh_r, loc=<42.5286 0.0019 0.7297>, rot=<0.0006 -0.0161 -0.0391 -0.9991>}
 calf_twist_01_r {parent=calf_r, loc=<35.8485 1.9197 0.5793>, rot=<0.0237 0.0095 -0.0035 -0.9997>}
 calf_twist_02_r {parent=calf_r, loc=<19.1546 1.7943 0.2638>, rot=<0.0237 0.0095 -0.0035 -0.9997>}
 foot_r {parent=calf_r, loc=<46.7403 2.5014 0.5919>, rot=<0.0427 -0.0238 0.0353 -0.9982>}
 ball_r {parent=foot_r, loc=<3.5658 -10.8066 0.1544>, rot=<0.0001 -0.0001 -0.7063 -0.7079>}
 knee_out_volume_base_r {parent=calf_r, loc=<0 -4 0>, rot=<0 0 0 1>}
 knee_out_volume_r {parent=knee_out_volume_base_r, loc=<0 0 0>, rot=<0 0 0 1>}
 knee_in_volume_base_r {parent=calf_r, loc=<-0.8031 5.0231 -0.2457>, rot=<0 0 0 -1>}
 knee_in_volume_r {parent=knee_in_volume_base_r, loc=<0 0 0>, rot=<0 0 0 1>}
 dyn_ankle_r {parent=calf_r, loc=<38.2914 2.8551 0.9919>, rot=<0.032 -0.0235 0.0356 -0.9986>}
 dyn_ball_r {parent=dyn_ankle_r, loc=<12.0273 -10.5393 0.3818>, rot=<0.0077 -0.0077 -0.7062 -0.7079>}
 thigh_twist_01_r {parent=thigh_r, loc=<27.952 0 0>, rot=<0.0013 0 0 -1>}
 glute_volume_base_r {parent=thigh_r, loc=<6 9 0>, rot=<0 0 -0.4617 -0.887>}
 glute_volume_r {parent=glute_volume_base_r, loc=<0 0 0>, rot=<0 0 0 -1>}
 crouch_volume_base_r {parent=thigh_r, loc=<-0.4799 -5.6776 0.2485>, rot=<0 0 0 -1>}
 crouch_volume_r {parent=crouch_volume_base_r, loc=<0 0 0>, rot=<0 0 0 -1>}
]
variables [
 elbow_rot_l {type=double, default=0}
 elbow_rot_r {type=double, default=0}
 knee_rot_l {type=double, default=0}
 knee_rot_r {type=double, default=0}
 psd_arm_bk_alpha_l {type=double, default=0}
 psd_arm_bk_alpha_r {type=double, default=0}
 psd_arm_fwd_alpha_l {type=double, default=0}
 psd_arm_fwd_alpha_r {type=double, default=0}
 psd_arm_fwd_dn_alpha_l {type=double, default=0}
 psd_arm_fwd_dn_alpha_r {type=double, default=0}
 psd_arm_fwd_up_alpha_l {type=double, default=0}
 psd_arm_fwd_up_alpha_r {type=double, default=0}
 psd_arm_up_alpha_l {type=double, default=0}
 psd_arm_up_alpha_r {type=double, default=0}
 thigh_rot_l {type=double, default=0}
 thigh_rot_r {type=double, default=0}
]
