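# Post-copy Anim Blueprint for the Bistro Astronaut female body, adding secondary motion and deformation to the pose copied from the parent mesh.
# The copied pose gets a neck layered blend, then RigidBody physics that switches off during skydiving, bone offsets and a muscle-volume Control Rig.
# The event graph feeds the rig limb rotations alongside arm pose-deformation query results.
# With a pistol animation set, left clavicle pitch drives the left arm plate offset.
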AnimBlueprint /BRCosmetics/Characters/Player/Female/Medium/Bodies/F_MED_Bistro_Astronaut/Meshes/F_MED_Bistro_Astronaut_AnimBP
target_skeleton=&/F_MED_Bistro_Astronaut_Skeleton
graphs [
 AnimGraph {nodes [
   n1 = RigidBody!(
    ComponentPose=Local To Component(LocalPose=Use cached pose 'incomingCachePose'())
    Alpha=Get Is Sky Diving() -> (Is Sky Diving:double)
    OverrideWorldGravity:Vector="0,0,-300"
    ExternalForce
   ) -> (Pose) {
    bClampLinearTranslationLimitToRefPose=true
    bOverrideWorldGravity=true
    overridePhysicsAsset=&/F_MED_Bistro_Astronaut_Physics
    simulationSpace=BaseBoneSpace
    alphaBoolBlend {blendOption=Linear}
    componentAppliedLinearAccClamp=<3000 3000 3000>
    componentLinearAccScale=<0.005 0.005 0.005>
    componentLinearVelScale=<0.005 0.005 0.005>
    overrideWorldGravity {z=-300}
   }
   n2 = "Transform (Modify) Bone - Bone: upperarm_twist_plate_01_l"(ComponentPose=@n1, Alpha=Get targetingHeadUp() -> (targetingHeadUp:double), Translation="-2,0,0", Rotation="-15,0,-60", Scale="1,1,1") {
    rotationMode=BMM_Additive
    rotationSpace=BCS_BoneSpace
    translationMode=BMM_Additive
    translationSpace=BCS_BoneSpace
    alphaBoolBlend {blendOption=Linear}
    alphaScaleBiasClamp {bMapRange=true, inRange {max=45, min=22}}
    boneToModify {boneName=upperarm_twist_plate_01_l}
    rotation {pitch=-15, roll=-60}
    translation {x=-2}
   }
   n3 = "Transform (Modify) Bone - Bone: dyn_gun_base"(ComponentPose=@n2, Alpha=1, Translation="0,0,-1", Rotation="0,0,0", Scale="1,1,1") {
    rotationMode=BMM_Additive
    rotationSpace=BCS_BoneSpace
    translationMode=BMM_Additive
    translationSpace=BCS_BoneSpace
    alphaBoolBlend {blendOption=Linear}
    alphaScaleBiasClamp {inRange {max=45, min=22}}
    boneToModify {boneName=dyn_gun_base}
    translation {z=-1}
   }
   n4 = Break Pose Deformation Query Results Arms!(PoseDeformationQueryResults_Arms:"Pose Deformation Query Results Arms Structure (by ref)"=Break Pose Deformation Query Arms!(PoseDeformationQuery_Arms:"Pose Deformation Query Arms Structure (by ref)"=Get Arms() -> (Arms:Pose Deformation Query Arms Structure)) -> (QueryResults:Pose Deformation Query Results Arms Structure) {is=title K2Node_BreakStruct}) -> (
    ArmUpL:float=~
    ArmUpR:float=~
    ArmForwardL:float=~
    ArmForwardR:float=~
    ArmBackL:float=~
    ArmBackR:float=~
    ArmForwardUpL:float=~
    ArmForwardUpR:float=~
    ArmForwardDownL:float=~
    ArmForwardDownR:float=~
   ) {is=title K2Node_BreakStruct}
   n5 = Control Rig!(
    Source=Component To Local(ComponentPose=@n3)
    Alpha
    thigh_rot_l:double=Get thigh_rot_l()
    elbow_rot_l:double=Get elbow_rot_l() -> (elbow_rot_l:double)
    elbow_rot_r:double=Get elbow_rot_r() -> (elbow_rot_r:double)
    thigh_rot_r:double=Get thigh_rot_r()
    knee_rot_l:double=Get knee_rot_l() -> (knee_rot_l:double)
    knee_rot_r:double=Get knee_rot_r() -> (knee_rot_r:double)
    psd_arm_up_alpha_l:double=@n4
    psd_arm_up_alpha_r:double=@n4.ArmUpR
    psd_arm_fwd_alpha_l:double=@n4.ArmForwardL
    psd_arm_fwd_alpha_r:double=@n4.ArmForwardR
    psd_arm_bk_alpha_l:double=@n4.ArmBackL
    psd_arm_bk_alpha_r:double=@n4.ArmBackR
    psd_arm_fwd_up_alpha_l:double=@n4.ArmForwardUpL
    psd_arm_fwd_up_alpha_r:double=@n4.ArmForwardUpR
    psd_arm_fwd_dn_alpha_l:double=@n4.ArmForwardDownL
    psd_arm_fwd_dn_alpha_r:double=@n4.ArmForwardDownR
   ) -> (Pose) {
    alphaBoolBlend {blendOption=Linear}
    controlRigAssetReference {blueprintRigClass=&/F_MED_Bistro_Muscle_Volume_ControlRig.F_MED_Bistro_Muscle_Volume_ControlRig_C}
   }
   Output Pose(Result=@n5) {layerGroup=None}
   Save cached pose 'headJointCache'(Pose=Copy Pose From Mesh() {bCopyCurves=true, bUseAttachedParent=true}) {cacheName=headJointCache}
   n6 = Layered blend per bone!(
    BasePose=Use cached pose 'headJointCache'()
    BlendPoses_0
    BlendPoses_1:Pose Link Structure=Use cached pose 'headJointCache'()
    BlendWeights_0=0.9
    BlendWeights_1:float=1
   ) -> (Pose) {
    blendPoses [{linkId=-1, sourceLinkId=-1}, {linkId=-1, sourceLinkId=-1}]
    blendWeights [1,1]
    layerSetup [{branchFilters [{blendDepth=1, boneName=neck_01}]}, {branchFilters [{blendDepth=1, boneName=head}]}]
   }
   Save cached pose 'incomingCachePose'(Pose=@n6) {cacheName=incomingCachePose}
  ]}
 EventGraph {nodes [
   Event BlueprintUpdateAnimation()
   > n1 = Sequence!(execute) -> (then_0, then_1, then_2:Exec, then_3:Exec, then_4:Exec, then_5:Exec, then_6:Exec)
   n2 = GetOwnerSkeletalMeshComponent()
   n3 = GetDeltaTransformFromRefPose(self=@n2, BoneName=clavicle_l, BaseName=spine_05)
   n4 = Set head_fwdBwd_rot(execute=@n1.then_3, head_fwdBwd_rot:double=BreakRotator(InRot=BreakTransform(InTransform=@n3).Rotation).Pitch) -> (Output_Get:double)
   n5 = Set Alt_weapon(execute=@n9.then, Alt_weapon:bool=true) -> (Output_Get:bool)
   n6 = Branch(execute=@n4.then, Condition=@n5)
   > Set targetingHeadUp(targetingHeadUp:double=@n4) -> (Output_Get:double)
   Set targetingHeadUp(execute=@n6.else, targetingHeadUp=0)
   n7 = Cast To FortPlayerAnimInstance(execute=@n1.then_4, Object=GetAnimInstance(self=@n2))
   > Set Alt_weapon(Alt_weapon=false)
   > n8 = Switch on EFortWeaponCoreAnimation!(execute, Selection:EFortWeaponCoreAnimation Enum=Get WeaponCoreAnim(self=@n7)) -> (
    Melee:Exec
    Pistol:Exec
    Shotgun:Exec
    PaperBlueprint:Exec
    Rifle:Exec
    MeleeOneHand:Exec
    MachinePistol:Exec
    RocketLauncher:Exec
    GrenadeLauncher:Exec
    GoingCommando:Exec
    AssaultRifle:Exec
    TacticalShotgun:Exec
    SniperRifle:Exec
    TrapPlacement:Exec
    ShoulderLauncher:Exec
    AbilityDecoTool:Exec
    Crossbow:Exec
    C4:Exec
    RemoteControl:Exec
    DualWield:Exec
    AR_BullPup:Exec
    AR_ForwardGrip:Exec
    MedPackPaddles:Exec
    SMG_P90:Exec
    AR_DrumGun:Exec
    Consumable_Small:Exec
    Consumable_Large:Exec
    Balloon:Exec
    MountedTurret:Exec
    CreativeTool:Exec
    ExplosiveBow:Exec
    AshtonIndigo:Exec
    AshtonChicago:Exec
    MeleeDualWield:Exec
    Unarmed:Exec
   )
   n9 = Branch(execute=@n8.Pistol)
   Set Alt_weapon(execute=@n9.else, Alt_weapon=false)
   n10 = Get bIsSkyDiving()
   Set Is Sky Diving(
    execute=@n1.then_5
    Is Sky Diving:double=Remap Float(Input="To Float (Boolean)"(InBool=@n10), OldRange=(X=0,Y=1), NewRange=(X=1,Y=0)) -> (Output=~)
   ) -> (Output_Get:double)
   n11 = GetDeltaTransformFromRefPose(self=GetOwnerSkeletalMeshComponent(), BoneName=lowerarm_l, BaseName=upperarm_l)
   Set elbow_rot_l(execute=@n1.then_0, elbow_rot_l:double=BreakRotator(InRot=BreakTransform(InTransform=@n11).Rotation).Yaw) -> (Output_Get:double)
   n12 = GetDeltaTransformFromRefPose(self=GetOwnerSkeletalMeshComponent(), BoneName=lowerarm_r, BaseName=upperarm_r)
   > Set elbow_rot_r(elbow_rot_r:double=BreakRotator(InRot=BreakTransform(InTransform=@n12).Rotation).Yaw) -> (Output_Get:double)
   n13 = GetDeltaTransformFromRefPose(self=GetOwnerSkeletalMeshComponent(), BoneName=calf_l, BaseName=thigh_l)
   Set knee_rot_l(execute=@n1.then_2, knee_rot_l:double=BreakRotator(InRot=BreakTransform(InTransform=@n13).Rotation).Yaw) -> (Output_Get:double)
   n14 = GetDeltaTransformFromRefPose(self=GetOwnerSkeletalMeshComponent(), BoneName=calf_r, BaseName=thigh_r)
   > Set knee_rot_r(knee_rot_r:double=BreakRotator(InRot=BreakTransform(InTransform=@n14).Rotation).Yaw) -> (Output_Get:double)
   n15 = GetDeltaTransformFromRefPose(self=GetOwnerSkeletalMeshComponent(), BoneName=thigh_l, BaseName=pelvis)
   Set thigh_rot_l(execute=@n1.then_1, thigh_rot_l=BreakRotator(InRot=BreakTransform(InTransform=@n15).Rotation).Yaw)
   n16 = GetDeltaTransformFromRefPose(self=GetOwnerSkeletalMeshComponent(), BoneName=thigh_r, BaseName=pelvis)
   > Set thigh_rot_r(thigh_rot_r=BreakRotator(InRot=BreakTransform(InTransform=@n16).Rotation).Yaw)
   TryGetPawnOwner()
   Set Gravity Override(Gravity Override=Gravity Override State Function DEPRECATED(
     CharacterPartAnimInstance=Get Character Part Anim Instance() -> (Character Part Anim Instance:Custom Character Part Anim Instance Object Reference)
     GameTimeInSeconds=GetGameTimeInSeconds()
     GravityOverrideParams=Get Gravity Override Params()
    ))
  ]}
]
variables [
 Alt_weapon {type=boolean}
 Arms {
  type=PoseDeformationQuery_Arms
  default {
   armBackL {angle=50, bDrawDebug=false, endJoint=lowerarm_l, poseJoint=pose_arm_bk_l, startJoint=upperarm_l}
   armBackR {angle=50, bDrawDebug=false, endJoint=lowerarm_r, poseJoint=pose_arm_bk_r, startJoint=upperarm_r}
   armForwardDownL {angle=40, bDrawDebug=false, endJoint=lowerarm_l, poseJoint=pose_arm_fwd_dn_l, startJoint=upperarm_l}
   armForwardDownR {angle=40, bDrawDebug=false, endJoint=lowerarm_r, poseJoint=pose_arm_fwd_dn_r, startJoint=upperarm_r}
   armForwardL {angle=50, bDrawDebug=false, endJoint=lowerarm_l, poseJoint=pose_arm_fwd_l, startJoint=upperarm_l}
   armForwardR {angle=50, bDrawDebug=false, endJoint=lowerarm_r, poseJoint=pose_arm_fwd_r, startJoint=upperarm_r}
   armForwardUpL {angle=45, bDrawDebug=false, endJoint=lowerarm_l, poseJoint=pose_arm_fwd_up_l, startJoint=upperarm_l}
   armForwardUpR {angle=45, bDrawDebug=false, endJoint=lowerarm_r, poseJoint=pose_arm_fwd_up_r, startJoint=upperarm_r}
   armUpL {angle=70, bDrawDebug=false, endJoint=lowerarm_l, poseJoint=pose_arm_up_l, startJoint=upperarm_l}
   armUpR {angle=70, bDrawDebug=false, endJoint=lowerarm_r, poseJoint=pose_arm_up_r, startJoint=upperarm_r}
  }
 }
 Character Part Anim Instance {type=CustomCharacterPartAnimInstance, default=null}
 elbow_rot_l {type=number}
 elbow_rot_r {type=number}
 Gravity Override {type=Vector}
 Gravity Override Params {
  type=GravityOverrideParamsStruct
  default {
   emoteOrMelee {gravityOverride=<0 0 -300>}
   front_End {gravityOverride=<0 0 -300>}
   onGround_Moving {gravityOverride=<0 0 -300>}
   onGround_Standing {gravityOverride=<0 0 -300>}
   skydive_DivingDown {gravityOverride=<0 0 0>, windAmplitude=<0 0 0>}
   skydive_FloatingDown {gravityOverride=<0 0 200>, windAmplitude=<0 0 0>, windFrequency=<0 0 0>, windOffset=<0 0 0>}
   skydive_GoingUp {gravityOverride=<0 0 300>, windAmplitude=<0 0 0>, windFrequency=<0 0 0>}
   skydive_Parachuting {gravityOverride=<0 0 -300>, windAmplitude=<0 0 0>, windFrequency=<0 0 0>}
  }
 }
 head_fwdBwd_rot {type=number}
 Is Sky Diving {type=number}
 knee_rot_l {type=number}
 knee_rot_r {type=number}
 result {type=boolean}
 targetingHeadUp {type=number}
 thigh_rot_l {type=number}
 thigh_rot_r {type=number}
]
