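# Control Rig for the face accessory part of the female medium Bistro Astronaut body, orienting head-offset and hair-band bones toward each other.
# An execution sequence first aims headOffset_base, the spine_05 child carrying the dyn_hair chain, at dyn_hair_band_base under the head.
# It then aims dyn_hair_band_base back at headOffset.
# Only the second aim weights its secondary axis, using a target with no bone space.
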ControlRigBlueprint /BRCosmetics/Characters/Player/Female/Medium/Bodies/F_MED_Bistro_Astronaut/Meshes/Parts/F_MED_Bistro_Astronaut_FaceAcc_CtrlRig
graphs [
 RigVMFunctionLibrary {links [], nodes []}
 RigVMModel {nodes [
   n1 = &RigUnit_BeginExecution()
   n2 = /Script/ControlRig.RigUnit_SequenceExecution(ExecuteContext=@n1.ExecutePin) {is=struct RigVMUnitNode}
   /Script/ControlRig.RigUnit_AimItem(
    DebugSettings=(Scale=10,WorldOffset=(Rotation=(X=0,Y=0,Z=0,W=1),Translation=(X=0,Y=0,Z=0),Scale3D=(X=1,Y=1,Z=1)))
    Item=(Type=Bone,Name="headOffset_base")
    Primary=(Weight=1,Axis=(X=1,Y=0,Z=0),Target=(X=1,Y=0,Z=0),Kind=Location,Space=(Type=Bone,Name="dyn_hair_band_base"))
    Secondary=(Weight=0,Axis=(X=0,Y=0,Z=1),Target=(X=1,Y=0,Z=0),Kind=Location,Space=(Type=Bone,Name="None"))
    Weight=1
    ExecutePin=@n2.A
   ) {is=struct RigVMUnitNode}
   &RigUnit_AimItem(
    DebugSettings=(Scale=10,WorldOffset=(Rotation=(X=0,Y=0,Z=0,W=1),Translation=(X=0,Y=0,Z=0),Scale3D=(X=1,Y=1,Z=1)))
    Item=(Type=Bone,Name="dyn_hair_band_base")
    Primary=(Weight=1,Axis=(X=1,Y=0,Z=0),Target=(X=1,Y=0,Z=0),Kind=Location,Space=(Type=Bone,Name="headOffset"))
    Secondary=(Weight=1,Axis=(X=0,Y=0,Z=1),Target=(X=1,Y=0,Z=0),Kind=Location,Space=(Type=Bone,Name="None"))
    Weight=1
    ExecutePin=@n2.B
   )
  ]}
]
hierarchy [
 root {loc=<0 0 0>, rot=<0 0 0 1>}
 pelvis {parent=root, loc=<0 -0.5301 98.0571>, rot=<0.0127 0.707 -0.0127 -0.707>}
 spine_01 {parent=pelvis, loc=<6.1208 -0.9328 0>, rot=<0 0 0.034 -0.9994>}
 spine_02 {parent=spine_01, loc=<6.5996 -0.2095 0>, rot=<0 0 0.0238 -0.9997>}
 spine_03 {parent=spine_02, loc=<6.4914 -0.1237 0>, rot=<0 0 0.0149 -0.9999>}
 spine_04 {parent=spine_03, loc=<6.9598 0.0794 0>, rot=<0 0 -0.0132 -0.9999>}
 spine_05 {parent=spine_04, loc=<6.9987 0 0>, rot=<0 0 -0.0015 -1>}
 clavicle_l {parent=spine_05, loc=<5.4831 -0.1053 -3.2946>, rot=<0.0465 -0.7289 -0.009 -0.6829>}
 upperarm_l {parent=clavicle_l, loc=<9.8658 -1.2136 -0.4105>, rot=<0.0164 -0.3245 -0.0269 -0.9454>}
 lowerarm_l {parent=upperarm_l, loc=<26.465 -0.8345 0.259>, rot=<0.0036 0.0161 -0.2298 -0.9731>}
 clavicle_r {parent=spine_05, loc=<5.4831 -0.1053 3.2946>, rot=<0.7289 0.0465 -0.6829 0.009>}
 upperarm_r {parent=clavicle_r, loc=<-9.8658 1.2136 0.4104>, rot=<0.0164 -0.3245 -0.0269 -0.9454>}
 lowerarm_r {parent=upperarm_r, loc=<-26.4651 0.8345 -0.259>, rot=<0.0036 0.0161 -0.2298 -0.9731>}
 neck_01 {parent=spine_05, loc=<11.3138 0.1529 0>, rot=<0 0 -0.2003 -0.9797>}
 neck_02 {parent=neck_01, loc=<4.5627 0 0>, rot=<0 0 0.0504 -0.9987>}
 head {parent=neck_02, loc=<5.3914 0 0>, rot=<0 0 0.115 -0.9934>}
 dyn_hair_band_base {parent=head, loc=<16.5082 -7.1218 0>, rot=<0.9522 -0.3055 0 0>}
 dyn_hairUpV {parent=head, loc=<25.1808 5.2031 0>, rot=<0.9534 -0.3016 0 0>}
 headOffset_base {parent=spine_05, loc=<25.4281 6.117 0>, rot=<0 0 -0.0401 -0.9992>}
 hairOffset {parent=headOffset_base, loc=<11.4875 -9.4355 0>, rot=<0.9509 -0.3094 0 0>}
 dyn_hair_band {parent=hairOffset, loc=<0.001 0.0056 0>, rot=<0 0 -0.0002 -1>}
 dyn_hair_a_base {parent=dyn_hair_band, loc=<11.7106 1.4768 0>, rot=<0 0 -0.7578 -0.6525>}
 dyn_hair_a {parent=dyn_hair_a_base, loc=<0 0 0>, rot=<0 0 -0.0805 -0.9968>}
 dyn_hair_b {parent=dyn_hair_a, loc=<20.5406 3.9638 0>, rot=<0 0 -0.3139 -0.9495>}
 dyn_hair_c {parent=dyn_hair_b, loc=<19.9023 2.6333 0>, rot=<0 0 -0.0377 -0.9993>}
 dyn_hair_d {parent=dyn_hair_c, loc=<21.5619 0 0>, rot=<-0.0111 -0.0297 0.1712 -0.9847>}
 dyn_hair_e {parent=dyn_hair_d, loc=<21.1802 -0.832 0>, rot=<-0.0077 0.1103 -0.0691 -0.9915>}
 dyn_hair_f {parent=dyn_hair_e, loc=<23.894 1.1315 0>, rot=<0 0 -0.4308 -0.9025>}
 dyn_hair_small_b {parent=dyn_hair_c, loc=<15.1013 5.4005 2.9992>, rot=<0 0 0.0377 -0.9993>}
 dyn_hair_small_c {parent=dyn_hair_c, loc=<6.3207 2.741 7.336>, rot=<0 0 0.0377 -0.9993>}
 dyn_hair_small_a {parent=dyn_hair_b, loc=<6.6273 -4.8786 -0.2055>, rot=<0 0 0 -1>}
 headOffset {parent=headOffset_base, loc=<0 0 0>, rot=<0 0 0 1>}
 thigh_l {parent=pelvis, loc=<-1.0855 -0.2109 -8.7053>, rot=<0.0009 0.0355 0.0155 -0.9993>}
 calf_l {parent=thigh_l, loc=<-42.5284 -0.0019 -0.7297>, rot=<0.0006 -0.0161 -0.0391 -0.9991>}
 thigh_r {parent=pelvis, loc=<-1.0853 -0.2109 8.7053>, rot=<0.0355 -0.0009 0.9993 0.0155>}
 calf_r {parent=thigh_r, loc=<42.5286 0.0019 0.7297>, rot=<0.0006 -0.0161 -0.0391 -0.9991>}
]
