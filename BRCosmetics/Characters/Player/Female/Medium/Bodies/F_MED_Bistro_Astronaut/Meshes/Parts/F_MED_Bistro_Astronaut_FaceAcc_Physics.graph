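# Physics asset for the face accessory mesh of the female medium Bistro Astronaut player body, setting up a dynamic hair chain with body collision shapes.
# Six capsule hair bodies link in series from a base body attached to the head offset, with moderate swing and tight twist limits.
# Capsules on the head, neck, spine, pelvis, arms and thighs approximate the body, joined by free-rotation constraints.
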PhysicsAsset /BRCosmetics/Characters/Player/Female/Medium/Bodies/F_MED_Bistro_Astronaut/Meshes/Parts/F_MED_Bistro_Astronaut_FaceAcc_Physics
bodies {
 dyn_hair_a [{shapeType=Capsule, length=11.203, radius=5, shapeName=dyn_hair_a_capsule, center=<6.4472 0 0>, rotation=<rot 90 0 0>}]
 dyn_hair_a_base [{shapeType=Capsule, length=0, radius=5.5314, shapeName=dyn_hair_a_base_capsule, rotation=<rot 90 0 0>}]
 dyn_hair_b [{shapeType=Capsule, length=13.0844, radius=5, shapeName=dyn_hair_b_capsule, center=<6.4472 0 0>, rotation=<rot 90 0 0>}]
 dyn_hair_c [{shapeType=Capsule, length=11.203, radius=5, shapeName=dyn_hair_c_capsule, center=<6.4472 0 0>, rotation=<rot 90 0 0>}]
 dyn_hair_d [{shapeType=Capsule, length=12.3354, radius=5.0009, shapeName=dyn_hair_d_capsule, center=<6.4472 0 0>, rotation=<rot 90 0 0>}]
 dyn_hair_e [{shapeType=Capsule, length=5, radius=5, shapeName=dyn_hair_e_capsule, center=<6.4472 0 0>, rotation=<rot 90 0 0>}]
 dyn_hair_f [{shapeType=Capsule, length=5, radius=5, shapeName=dyn_hair_f_capsule, center=<6.4472 0 0>, rotation=<rot 90 0 0>}]
 headOffset_base [{shapeType=Capsule, length=0.505, radius=14.4062, shapeName=headOffset_base_capsule, rotation=<rot 0 -90 0>}]
 lowerarm_l [{shapeType=Capsule, length=18.5681, radius=4.8844, shapeName=lowerarm_l_capsule, center=<12.4436 -0.0002 0>, rotation=<rot 90 0 0>}]
 lowerarm_r [{shapeType=Capsule, length=18.5681, radius=4.8844, shapeName=lowerarm_r_capsule, center=<-12.4436 -0.0002 0>, rotation=<rot 90 0 0>}]
 neck_01 [{shapeType=Capsule, length=15.8353, radius=10.9812, shapeName=neck_01_capsule, rotation=<rot 0 90 180>}]
 pelvis [{shapeType=Capsule, length=5.7913, radius=12.282, shapeName=pelvis_capsule}]
 spine_01 [{
   shapeType=Capsule
   length=16.7366
   radius=12.9304
   shapeName=spine_01_capsule
   center=<10.6558 0 -0.4711>
   rotation=<rot 0 -89.9999 -87.4706>
  }]
 spine_05 [{shapeType=Capsule, length=28.8086, radius=11.145, shapeName=spine_05_capsule}]
 thigh_l [{shapeType=Capsule, length=22.3123, radius=9.7393, shapeName=thigh_l_capsule, center=<-21.2798 0 0>, rotation=<rot 90 0 0>}]
 thigh_r [{shapeType=Capsule, length=22.3123, radius=9.7393, shapeName=thigh_r_capsule, center=<21.2798 0 0>, rotation=<rot 90 0 0>}]
 upperarm_l [{shapeType=Capsule, length=18.5681, radius=4.8844, shapeName=upperarm_l_capsule, center=<12.4436 -0.0002 0>, rotation=<rot 90 0 0>}]
 upperarm_r [{shapeType=Capsule, length=18.5681, radius=4.8844, shapeName=upperarm_r_capsule, center=<-12.4436 -0.0002 0>, rotation=<rot 90 0 0>}]
}
constraints [
 {
  bone1Name=dyn_hair_a_base
  bone2Name=headOffset_base
  swing1LimitDegrees=45
  swing1Motion=Free
  swing2LimitDegrees=45
  swing2Motion=Free
  twistLimitDegrees=45
  twistMotion=Free
 }
 {
  bone1Name=spine_05
  bone2Name=spine_01
  swing1LimitDegrees=45
  swing1Motion=Free
  swing2LimitDegrees=45
  swing2Motion=Free
  twistLimitDegrees=45
  twistMotion=Free
 }
 {
  bone1Name=upperarm_l
  bone2Name=spine_05
  swing1LimitDegrees=45
  swing1Motion=Free
  swing2LimitDegrees=45
  swing2Motion=Free
  twistLimitDegrees=45
  twistMotion=Free
 }
 {
  bone1Name=lowerarm_l
  bone2Name=upperarm_l
  swing1LimitDegrees=45
  swing1Motion=Free
  swing2LimitDegrees=45
  swing2Motion=Free
  twistLimitDegrees=45
  twistMotion=Free
 }
 {
  bone1Name=upperarm_r
  bone2Name=spine_05
  swing1LimitDegrees=45
  swing1Motion=Free
  swing2LimitDegrees=45
  swing2Motion=Free
  twistLimitDegrees=45
  twistMotion=Free
 }
 {
  bone1Name=lowerarm_r
  bone2Name=upperarm_r
  swing1LimitDegrees=45
  swing1Motion=Free
  swing2LimitDegrees=45
  swing2Motion=Free
  twistLimitDegrees=45
  twistMotion=Free
 }
 {
  bone1Name=thigh_l
  bone2Name=pelvis
  swing1LimitDegrees=45
  swing1Motion=Free
  swing2LimitDegrees=45
  swing2Motion=Free
  twistLimitDegrees=45
  twistMotion=Free
 }
 {
  bone1Name=thigh_r
  bone2Name=pelvis
  swing1LimitDegrees=45
  swing1Motion=Free
  swing2LimitDegrees=45
  swing2Motion=Free
  twistLimitDegrees=45
  twistMotion=Free
 }
 {bone1Name=dyn_hair_a, bone2Name=dyn_hair_a_base, swing1LimitDegrees=30, swing2LimitDegrees=30, twistLimitDegrees=5}
 {bone1Name=dyn_hair_b, bone2Name=dyn_hair_a, swing1LimitDegrees=30, swing2LimitDegrees=30, twistLimitDegrees=5}
 {bone1Name=dyn_hair_c, bone2Name=dyn_hair_b, swing1LimitDegrees=30, swing2LimitDegrees=30, twistLimitDegrees=5}
 {bone1Name=dyn_hair_d, bone2Name=dyn_hair_c, swing1LimitDegrees=30, swing2LimitDegrees=30, twistLimitDegrees=5}
 {bone1Name=dyn_hair_e, bone2Name=dyn_hair_d, swing1LimitDegrees=30, swing2LimitDegrees=30, twistLimitDegrees=5}
 {bone1Name=dyn_hair_f, bone2Name=dyn_hair_e, swing1LimitDegrees=30, swing2LimitDegrees=30, twistLimitDegrees=5}
 {
  bone1Name=spine_01
  bone2Name=pelvis
  swing1LimitDegrees=45
  swing1Motion=Free
  swing2LimitDegrees=45
  swing2Motion=Free
  twistLimitDegrees=45
  twistMotion=Free
 }
 {
  bone1Name=headOffset_base
  bone2Name=spine_05
  swing1LimitDegrees=45
  swing1Motion=Free
  swing2LimitDegrees=45
  swing2Motion=Free
  twistLimitDegrees=45
  twistMotion=Free
 }
 {
  bone1Name=neck_01
  bone2Name=spine_05
  swing1LimitDegrees=45
  swing1Motion=Free
  swing2LimitDegrees=45
  swing2Motion=Free
  twistLimitDegrees=45
  twistMotion=Free
 }
]
