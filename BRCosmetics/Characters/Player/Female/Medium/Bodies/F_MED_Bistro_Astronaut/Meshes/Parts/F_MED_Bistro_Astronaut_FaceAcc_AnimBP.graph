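# Post-process style AnimBP for the F_MED_Bistro_Astronaut face accessory that drives its dynamic hair bones.
# It copies parent and body poses, runs a Control Rig, then layers bone offsets, constraints and dynamics into a shared cached pose.
# That pose feeds a RigidBody/AnimDynamics branch and a Trail-controller branch, blended by bRigidBodyEnabled.
# The EventGraph sets correction and simulation alphas from backbling type, glider tags and skydiving state.
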AnimBlueprint /BRCosmetics/Characters/Player/Female/Medium/Bodies/F_MED_Bistro_Astronaut/Meshes/Parts/F_MED_Bistro_Astronaut_FaceAcc_AnimBP
target_skeleton=&/F_MED_Bistro_Astronaut_FaceAcc_Skeleton
graphs [
 AnimGraph {nodes [
   n1 = Get head_fwdBwd_rot() -> (head_fwdBwd_rot:double)
   n2 = "Transform (Modify) Bone - Bone: dyn_hair_a_base"(Alpha=@n1, Translation="0,0,0", Rotation="0,-50,0", Scale="1,1,1") {
    rotationMode=BMM_Additive
    rotationSpace=BCS_BoneSpace
    scaleMode=BMM_Additive
    scaleSpace=BCS_BoneSpace
    translationMode=BMM_Additive
    translationSpace=BCS_BoneSpace
    alphaBoolBlend {blendOption=Linear}
    alphaScaleBiasClamp {bMapRange=true, inRange {max=0, min=-30}, outRange {max=0, min=1}}
    boneToModify {boneName=dyn_hair_a_base}
    rotation {yaw=-50}
   }
   "Transform (Modify) Bone - Bone: dyn_hair_a_base"(ComponentPose=@n2, Alpha=@n1, Translation="0,0,0", Rotation="0,10,0", Scale="1,1,1") {
    rotationMode=BMM_Additive
    rotationSpace=BCS_BoneSpace
    scaleMode=BMM_Additive
    scaleSpace=BCS_BoneSpace
    translationMode=BMM_Additive
    translationSpace=BCS_BoneSpace
    alphaBoolBlend {blendOption=Linear}
    alphaScaleBiasClamp {bMapRange=true, inRange {max=30}}
    boneToModify {boneName=dyn_hair_a_base}
    rotation {yaw=10}
   }
   Output Pose(Result=Blend Poses by bool(
     BlendPose_0=Use cached pose 'simCachePose'()
     BlendPose_1=Use cached pose 'noSimCachePos'()
     BlendTime_1=0.25
     bActiveValue=Get bRigidBodyEnabled()
    ) {blendType=Linear}) {layerGroup=None}
   n3 = Layered blend per bone(
    BasePose=Copy Pose From Mesh() {bCopyCurves=true, bUseAttachedParent=true}
    BlendPoses_0=Copy Pose From Mesh(SourceMeshComponent=Get Body_Mesh())
   ) {layerSetup [{branchFilters [
       {boneName=upperarm_l}
       {boneName=upperarm_r}
       {boneName=lowerarm_l}
       {boneName=lowerarm_r}
       {boneName=thigh_l}
       {boneName=thigh_r}
       {boneName=calf_l}
       {boneName=calf_r}
      ]}]}
   n4 = Control Rig(Source=@n3) {
    alphaBoolBlend {blendOption=Linear}
    controlRigAssetReference {blueprintRigClass=&/F_MED_Bistro_Astronaut_FaceAcc_CtrlRig.F_MED_Bistro_Astronaut_FaceAcc_CtrlRig_C}
   }
   n5 = Get head_fwdBwd_rot()
   n6 = "Transform (Modify) Bone - Bone: dyn_hair_a_base"(ComponentPose=Local To Component(LocalPose=@n4), Alpha=@n5, Translation="0,0,0", Rotation="0,-60,0", Scale="1,1,1") {
    rotationMode=BMM_Additive
    rotationSpace=BCS_BoneSpace
    scaleMode=BMM_Additive
    scaleSpace=BCS_BoneSpace
    translationMode=BMM_Additive
    translationSpace=BCS_BoneSpace
    alphaBoolBlend {blendOption=Linear}
    alphaScaleBiasClamp {bMapRange=true, inRange {max=0, min=-30}, outRange {max=0, min=1}}
    boneToModify {boneName=dyn_hair_a_base}
    rotation {yaw=-60}
   }
   n7 = "Transform (Modify) Bone - Bone: dyn_hair_a_base"(ComponentPose=@n6, Alpha=@n5, Translation="0,0,0", Rotation="0,10,0", Scale="1,1,1") {
    rotationMode=BMM_Additive
    rotationSpace=BCS_BoneSpace
    scaleMode=BMM_Additive
    scaleSpace=BCS_BoneSpace
    translationMode=BMM_Additive
    translationSpace=BCS_BoneSpace
    alphaBoolBlend {blendOption=Linear}
    alphaScaleBiasClamp {bMapRange=true, bias=1, scale=0, inRange {max=30}}
    boneToModify {boneName=dyn_hair_a_base}
    rotation {yaw=10}
   }
   n8 = "Transform (Modify) Bone - Bone: dyn_hair_b"(ComponentPose=@n7, Alpha=@n5, Translation="0,0,0", Rotation="0,90,0", Scale="1,1,1") {
    rotationMode=BMM_Additive
    rotationSpace=BCS_BoneSpace
    scaleMode=BMM_Additive
    scaleSpace=BCS_BoneSpace
    translationMode=BMM_Additive
    translationSpace=BCS_BoneSpace
    alphaBoolBlend {blendOption=Linear}
    alphaScaleBiasClamp {bMapRange=true, bias=1, scale=0, inRange {max=20}}
    boneToModify {boneName=dyn_hair_b}
    rotation {yaw=90}
   }
   n9 = RigidBody!(ComponentPose=@n8, Alpha, OverrideWorldGravity:Vector=Get gravityOverride() -> (gravityOverride:Vector), ExternalForce) -> (Pose) {
    overridePhysicsAsset=&/F_MED_Bistro_Astronaut_FaceAcc_Physics
    alphaBoolBlend {blendOption=Linear}
    overrideWorldGravity {z=-980}
   }
   n10 = Anim Dynamics - Bone: dyn_hair_small_a(ComponentPose=@n9, Alpha=0.7, ExternalForce="0,0,0") {
    angularDampingOverride=1
    angularSpringConstant=20000
    bAngularSpring=true
    bGravityOverrideInSimSpace=true
    bOverrideAngularDamping=true
    bUseGravityOverride=true
    gravityScale=0.35
    simulationSpace=RootRelative
    alphaBoolBlend {blendOption=Linear}
    alphaScaleBiasClamp {outRange {max=0, min=0.5}}
    boundBone {boneName=dyn_hair_small_a}
    componentAppliedLinearAccClamp=<3000 3000 3000>
    componentLinearAccScale=<1 1 1>
    componentLinearVelScale=<1 1 1>
    gravityOverride {z=-980}
    physicsBodyDefinitions [{
      sphereCollisionRadius=0
      boundBone {boneName=dyn_hair_small_a}
      boxExtents=<20 30 30>
      constraintSetup {angularLimitsMax=<0 0 0>, angularLimitsMin=<0 -30 -45>, angularTarget=<1 0 0>}
      localJointOffset=<7 0 0>
     }]
   }
   n11 = Anim Dynamics - Bone: dyn_hair_small_b(ComponentPose=@n10, Alpha=0.9, ExternalForce="0,0,0") {
    angularDampingOverride=1
    angularSpringConstant=20000
    bAngularSpring=true
    bGravityOverrideInSimSpace=true
    bOverrideAngularDamping=true
    bUseGravityOverride=true
    gravityScale=0.35
    simulationSpace=RootRelative
    alphaBoolBlend {blendOption=Linear}
    alphaScaleBiasClamp {outRange {max=0, min=0.9}}
    boundBone {boneName=dyn_hair_small_b}
    componentAppliedLinearAccClamp=<3000 3000 3000>
    componentLinearAccScale=<1 1 1>
    componentLinearVelScale=<1 1 1>
    gravityOverride {z=-980}
    physicsBodyDefinitions [{
      sphereCollisionRadius=0
      boundBone {boneName=dyn_hair_small_b}
      boxExtents=<20 30 30>
      constraintSetup {angularLimitsMax=<0 0 0>, angularLimitsMin=<0 -45 0>, angularTarget=<1 0 0>}
      localJointOffset=<7 0 0>
     }]
   }
   n12 = Anim Dynamics - Bone: dyn_hair_small_c(ComponentPose=@n11, Alpha=0.8, ExternalForce="0,0,0") {
    angularDampingOverride=1
    angularSpringConstant=20000
    bAngularSpring=true
    bGravityOverrideInSimSpace=true
    bOverrideAngularDamping=true
    bUseGravityOverride=true
    gravityScale=0.35
    simulationSpace=RootRelative
    alphaBoolBlend {blendOption=Linear}
    alphaScaleBiasClamp {outRange {max=0, min=0.7}}
    boundBone {boneName=dyn_hair_small_c}
    componentAppliedLinearAccClamp=<3000 3000 3000>
    componentLinearAccScale=<1 1 1>
    componentLinearVelScale=<1 1 1>
    gravityOverride {z=-980}
    physicsBodyDefinitions [{
      sphereCollisionRadius=0
      boundBone {boneName=dyn_hair_small_c}
      boxExtents=<20 30 30>
      constraintSetup {angularLimitsMax=<15 0 0>, angularLimitsMin=<-15 -45 -45>, angularTarget=<1 0 0>}
      localJointOffset=<7 0 0>
     }]
   }
   n13 = Get hairConstraintON() -> (hairConstraintON:bool)
   n14 = "Constraint - dyn_hair_a (PP)"!(ComponentPose=@n12, bAlphaBoolEnabled:bool=@n13, ConstraintWeights_0=0.7, ConstraintWeights_1:float=0.3) -> (Pose) {
    alphaInputType=Bool
    alphaBoolBlend {blendOption=Linear}
    boneToModify {boneName=dyn_hair_a}
    constraintSetup [{targetBone {boneName=head}}, {targetBone {boneName=spine_05}}]
    constraintWeights [1,1]
   }
   n15 = "Constraint - dyn_hair_b (PP)"!(ComponentPose=@n14, bAlphaBoolEnabled:bool=@n13, ConstraintWeights_0=0.3, ConstraintWeights_1:float=0.7) -> (Pose) {
    alphaInputType=Bool
    alphaBoolBlend {blendOption=Linear}
    boneToModify {boneName=dyn_hair_b}
    constraintSetup [{targetBone {boneName=head}}, {targetBone {boneName=spine_05}}]
    constraintWeights [1,1]
   }
   n16 = "Constraint - dyn_hair_c (PPP)"!(ComponentPose=@n15, bAlphaBoolEnabled:bool=@n13, ConstraintWeights_0=0.2, ConstraintWeights_1:float=0.2, ConstraintWeights_2:float=0.6) -> (Pose) {
    alphaInputType=Bool
    alphaBoolBlend {blendOption=Linear}
    boneToModify {boneName=dyn_hair_c}
    constraintSetup [{targetBone {boneName=pelvis}}, {targetBone {boneName=spine_04}}, {targetBone {boneName=spine_05}}]
    constraintWeights [1,1,1]
   }
   n17 = "Constraint - dyn_hair_d (PPP)"!(ComponentPose=@n16, bAlphaBoolEnabled:bool=@n13, ConstraintWeights_0=0.1, ConstraintWeights_1:float=0.7, ConstraintWeights_2:float=0.2) -> (Pose) {
    alphaInputType=Bool
    alphaBoolBlend {blendOption=Linear}
    boneToModify {boneName=dyn_hair_d}
    constraintSetup [{targetBone {boneName=pelvis}}, {targetBone {boneName=spine_04}}, {targetBone {boneName=spine_05}}]
    constraintWeights [1,1,1]
   }
   n18 = "Constraint - dyn_hair_e (PP)"!(ComponentPose=@n17, bAlphaBoolEnabled:bool=@n13, ConstraintWeights_0=0.1, ConstraintWeights_1:float=0.9) -> (Pose) {
    alphaInputType=Bool
    alphaBoolBlend {blendOption=Linear}
    boneToModify {boneName=dyn_hair_e}
    constraintSetup [{targetBone {boneName=pelvis}}, {targetBone {boneName=spine_02}}]
    constraintWeights [1,1]
   }
   n19 = Get hairConstraintON()
   n20 = "Transform (Modify) Bone - Bone: dyn_hair_b"!(ComponentPose=@n18, bAlphaBoolEnabled:bool=@n19, Translation="0,0,-5", Rotation="5,0,0", Scale="1,1,1") -> (Pose) {
    alphaInputType=Bool
    rotationMode=BMM_Additive
    rotationSpace=BCS_BoneSpace
    translationMode=BMM_Additive
    translationSpace=BCS_BoneSpace
    alphaBoolBlend {blendOption=Linear}
    alphaScaleBiasClamp {bMapRange=true, inRange {max=0, min=-25}, outRange {max=0, min=1}}
    boneToModify {boneName=dyn_hair_b}
    rotation {pitch=5}
    translation {z=-5}
   }
   n21 = "Constraint - dyn_hair_small_a (PPP)"!(ComponentPose=@n20, bAlphaBoolEnabled:bool=@n19, ConstraintWeights_0=0.2, ConstraintWeights_1:float=0.7, ConstraintWeights_2:float=0.1) -> (Pose) {
    alphaInputType=Bool
    alphaBoolBlend {blendOption=Linear}
    boneToModify {boneName=dyn_hair_small_a}
    constraintSetup [{targetBone {boneName=dyn_hair_c}}, {targetBone {boneName=dyn_hair_b}}, {targetBone {boneName=dyn_hair_a}}]
    constraintWeights [1,1,1]
   }
   n22 = "Constraint - dyn_hair_small_b (PP)"!(ComponentPose=@n21, bAlphaBoolEnabled:bool=@n19, ConstraintWeights_0=0.3, ConstraintWeights_1:float=0.7) -> (Pose) {
    alphaInputType=Bool
    alphaBoolBlend {blendOption=Linear}
    boneToModify {boneName=dyn_hair_small_b}
    constraintSetup [{targetBone {boneName=dyn_hair_d}}, {targetBone {boneName=dyn_hair_c}}]
    constraintWeights [1,1]
   }
   n23 = "Constraint - dyn_hair_small_c (PPP)"!(ComponentPose=@n22, bAlphaBoolEnabled:bool=@n19, ConstraintWeights_0=0.1, ConstraintWeights_1:float=0.8, ConstraintWeights_2:float=0.1) -> (Pose) {
    alphaInputType=Bool
    alphaBoolBlend {blendOption=Linear}
    boneToModify {boneName=dyn_hair_small_c}
    constraintSetup [{targetBone {boneName=dyn_hair_d}}, {targetBone {boneName=dyn_hair_c}}, {targetBone {boneName=dyn_hair_b}}]
    constraintWeights [1,1,1]
   }
   n24 = Get IsShortBackblingOn() -> (IsShortBackblingOn:bool)
   n25 = "Transform (Modify) Bone - Bone: dyn_hair_a"!(ComponentPose=@n23, bAlphaBoolEnabled:bool=@n24, Translation="0,0,0", Rotation="0,9.5,0", Scale="1,1,1") -> (Pose) {
    alphaInputType=Bool
    rotationMode=BMM_Additive
    rotationSpace=BCS_BoneSpace
    translationMode=BMM_Additive
    translationSpace=BCS_BoneSpace
    alphaBoolBlend {blendOption=Linear}
    alphaScaleBiasClamp {bMapRange=true, inRange {max=0, min=-25}, outRange {max=0, min=1}}
    boneToModify {boneName=dyn_hair_a}
    rotation {yaw=9.5}
   }
   n26 = "Transform (Modify) Bone - Bone: dyn_hair_b"(ComponentPose=@n25, bAlphaBoolEnabled=@n24, Translation="0,2.5,0", Rotation="0,-9,0", Scale="1,1,1") {
    alphaInputType=Bool
    rotationMode=BMM_Additive
    rotationSpace=BCS_BoneSpace
    translationMode=BMM_Additive
    translationSpace=BCS_BoneSpace
    alphaBoolBlend {blendOption=Linear}
    alphaScaleBiasClamp {bMapRange=true, inRange {max=0, min=-25}, outRange {max=0, min=1}}
    boneToModify {boneName=dyn_hair_b}
    rotation {yaw=-9}
    translation {y=2.5}
   }
   n27 = "Transform (Modify) Bone - Bone: dyn_hair_c"!(ComponentPose=@n26, bAlphaBoolEnabled:bool=@n24, Translation="0,2.5,0", Rotation="0,5,0", Scale="1,1,1") -> (Pose) {
    alphaInputType=Bool
    rotationMode=BMM_Additive
    rotationSpace=BCS_BoneSpace
    translationMode=BMM_Additive
    translationSpace=BCS_BoneSpace
    alphaBoolBlend {blendOption=Linear}
    alphaScaleBiasClamp {bMapRange=true, inRange {max=0, min=-25}, outRange {max=0, min=1}}
    boneToModify {boneName=dyn_hair_c}
    rotation {yaw=5}
    translation {y=2.5}
   }
   n28 = Get IsShortBackblingOn()
   n29 = "Transform (Modify) Bone - Bone: dyn_hair_small_a"!(ComponentPose=@n27, bAlphaBoolEnabled:bool=@n28, Translation="0,2.5,0", Rotation="-10,0,0", Scale="1,1,1") -> (Pose) {
    alphaInputType=Bool
    rotationMode=BMM_Additive
    rotationSpace=BCS_BoneSpace
    translationMode=BMM_Additive
    translationSpace=BCS_BoneSpace
    alphaBoolBlend {blendOption=Linear}
    alphaScaleBiasClamp {bMapRange=true, inRange {max=0, min=-25}, outRange {max=0, min=1}}
    boneToModify {boneName=dyn_hair_small_a}
    rotation {pitch=-10}
    translation {y=2.5}
   }
   n30 = "Transform (Modify) Bone - Bone: dyn_hair_small_b"!(ComponentPose=@n29, bAlphaBoolEnabled:bool=@n28, Translation="0,-4,0", Rotation="0,5,0", Scale="1,1,1") -> (Pose) {
    alphaInputType=Bool
    rotationMode=BMM_Additive
    rotationSpace=BCS_BoneSpace
    translationMode=BMM_Additive
    translationSpace=BCS_BoneSpace
    alphaBoolBlend {blendOption=Linear}
    alphaScaleBiasClamp {bMapRange=true, inRange {max=0, min=-25}, outRange {max=0, min=1}}
    boneToModify {boneName=dyn_hair_small_b}
    rotation {yaw=5}
    translation {y=-4}
   }
   n31 = "Transform (Modify) Bone - Bone: dyn_hair_small_c"!(ComponentPose=@n30, bAlphaBoolEnabled:bool=@n28, Translation="0,-3,0", Rotation="0,0,0", Scale="1,1,1") -> (Pose) {
    alphaInputType=Bool
    rotationMode=BMM_Additive
    rotationSpace=BCS_BoneSpace
    translationMode=BMM_Additive
    translationSpace=BCS_BoneSpace
    alphaBoolBlend {blendOption=Linear}
    alphaScaleBiasClamp {bMapRange=true, inRange {max=0, min=-25}, outRange {max=0, min=1}}
    boneToModify {boneName=dyn_hair_small_c}
    translation {y=-3}
   }
   n32 = "Transform (Modify) Bone - Bone: dyn_hair_d"!(ComponentPose=@n31, bAlphaBoolEnabled:bool=@n28, Translation="0,-1,0", Rotation="0,-5,0", Scale="1,1,1") -> (Pose) {
    alphaInputType=Bool
    rotationMode=BMM_Additive
    rotationSpace=BCS_BoneSpace
    translationMode=BMM_Additive
    translationSpace=BCS_BoneSpace
    alphaBoolBlend {blendOption=Linear}
    alphaScaleBiasClamp {bMapRange=true, inRange {max=0, min=-25}, outRange {max=0, min=1}}
    boneToModify {boneName=dyn_hair_d}
    rotation {yaw=-5}
    translation {y=-1}
   }
   n33 = Get backpackOnHeadUpDn() -> (backpackOnHeadUpDn:double)
   n34 = "Transform (Modify) Bone - Bone: dyn_hair_a"!(ComponentPose=@n32, Alpha:float=@n33, Translation="0,2,0", Rotation="0,0,0", Scale="1,1,1") -> (Pose) {
    rotationMode=BMM_Additive
    rotationSpace=BCS_BoneSpace
    translationMode=BMM_Additive
    translationSpace=BCS_BoneSpace
    alphaBoolBlend {blendOption=Linear}
    alphaScaleBiasClamp {bMapRange=true, inRange {max=0, min=-25}, outRange {max=0, min=1}}
    boneToModify {boneName=dyn_hair_a}
    translation {y=2}
   }
   n35 = "Transform (Modify) Bone - Bone: dyn_hair_b"!(ComponentPose=@n34, Alpha:float=@n33, Translation="5,-5,0", Rotation="0,5,0", Scale="1,1,1") -> (Pose) {
    rotationMode=BMM_Additive
    rotationSpace=BCS_BoneSpace
    translationMode=BMM_Additive
    translationSpace=BCS_BoneSpace
    alphaBoolBlend {blendOption=Linear}
    alphaScaleBiasClamp {bMapRange=true, inRange {max=0, min=-25}, outRange {max=0, min=1}}
    boneToModify {boneName=dyn_hair_b}
    rotation {yaw=5}
    translation=<5 -5>
   }
   n36 = "Transform (Modify) Bone - Bone: dyn_hair_small_b"!(ComponentPose=@n35, Alpha:float=@n33, Translation="0,-1.5,0", Rotation="0,0,0", Scale="1,1,1") -> (Pose) {
    rotationMode=BMM_Additive
    rotationSpace=BCS_BoneSpace
    translationMode=BMM_Additive
    translationSpace=BCS_BoneSpace
    alphaBoolBlend {blendOption=Linear}
    alphaScaleBiasClamp {bMapRange=true, inRange {max=0, min=-25}, outRange {max=0, min=1}}
    boneToModify {boneName=dyn_hair_small_b}
    translation {y=-1.5}
   }
   n37 = Get backpackOnHeadUpDn()
   n38 = "Transform (Modify) Bone - Bone: dyn_hair_a"(ComponentPose=@n36, Alpha=@n37, Translation="-2,1,0", Rotation="0,-2,0", Scale="1,1,1") {
    rotationMode=BMM_Additive
    rotationSpace=BCS_BoneSpace
    translationMode=BMM_Additive
    translationSpace=BCS_BoneSpace
    alphaBoolBlend {blendOption=Linear}
    alphaScaleBiasClamp {bMapRange=true, inRange {max=15}}
    boneToModify {boneName=dyn_hair_a}
    rotation {yaw=-2}
    translation=<-2 1>
   }
   n39 = "Transform (Modify) Bone - Bone: dyn_hair_b"(ComponentPose=@n38, Alpha=@n37, Translation="-4,0,0", Rotation="0,0,0", Scale="1,1,1") {
    rotationMode=BMM_Additive
    rotationSpace=BCS_BoneSpace
    translationMode=BMM_Additive
    translationSpace=BCS_BoneSpace
    alphaBoolBlend {blendOption=Linear}
    alphaScaleBiasClamp {bMapRange=true, inRange {max=15}}
    boneToModify {boneName=dyn_hair_b}
    translation {x=-4}
   }
   n40 = "Transform (Modify) Bone - Bone: dyn_hair_small_b"(ComponentPose=@n39, Alpha=@n37, Translation="0,0,0", Rotation="0,5,0", Scale="1,1,1") {
    rotationMode=BMM_Additive
    rotationSpace=BCS_BoneSpace
    translationMode=BMM_Additive
    translationSpace=BCS_BoneSpace
    alphaBoolBlend {blendOption=Linear}
    alphaScaleBiasClamp {bMapRange=true, inRange {max=15}}
    boneToModify {boneName=dyn_hair_small_b}
    rotation {yaw=5}
   }
   n41 = Get isBistroGliderAlpha() -> (isBistroGliderAlpha:double)
   n42 = "Transform (Modify) Bone - Bone: dyn_hair_a"(ComponentPose=@n40, Alpha=@n41, Translation="0,0,0", Rotation="0,-35,0", Scale="1,1,1") {
    rotationMode=BMM_Additive
    rotationSpace=BCS_BoneSpace
    translationMode=BMM_Additive
    translationSpace=BCS_BoneSpace
    alphaBoolBlend {blendOption=Linear}
    alphaScaleBiasClamp {bMapRange=true}
    boneToModify {boneName=dyn_hair_a}
    rotation {yaw=-35}
   }
   n43 = "Transform (Modify) Bone - Bone: dyn_hair_b"(ComponentPose=@n42, Alpha=@n41, Translation="0,0,0", Rotation="0,-20,0", Scale="1,1,1") {
    rotationMode=BMM_Additive
    rotationSpace=BCS_BoneSpace
    translationMode=BMM_Additive
    translationSpace=BCS_BoneSpace
    alphaBoolBlend {blendOption=Linear}
    alphaScaleBiasClamp {bMapRange=true}
    boneToModify {boneName=dyn_hair_b}
    rotation {yaw=-20}
   }
   n44 = "Transform (Modify) Bone - Bone: dyn_hair_a_base"(ComponentPose=@n43, Alpha=@n41, Translation="0,0,0", Rotation="0,-10,0", Scale="1,1,1") {
    rotationMode=BMM_Additive
    rotationSpace=BCS_BoneSpace
    translationMode=BMM_Additive
    translationSpace=BCS_BoneSpace
    alphaBoolBlend {blendOption=Linear}
    alphaScaleBiasClamp {bMapRange=true}
    boneToModify {boneName=dyn_hair_a_base}
    rotation {yaw=-10}
   }
   Save cached pose 'sharedCachePos'(Pose=Component To Local(ComponentPose=@n44)) {cacheName=sharedCachePos}
   n45 = Trail controller - Bone: dyn_hair_f!(
    ComponentPose=Local To Component(LocalPose=Use cached pose 'sharedCachePos'())
    Alpha=Get TrailAlpha() -> (TrailAlpha:double)
    RelaxationSpeedScale:float=Get RelaxSpeed() -> (RelaxSpeed:double)
   ) -> (Pose) {
    bLimitRotation=true
    bShowBaseMotion=false
    bShowLimit=false
    bUsePlanarLimit=true
    chainLength=6
    debugLifeTime=1
    maxDeltaTime=0.0333
    alphaScaleBiasClamp {bInterpResult=true, interpSpeedIncreasing=20}
    planarLimits [{drivingBone {boneName=spine_05}, planeTransform {location=<5 0 0>, rotation=<rot 90 90 0>, scale=<1 1 1>}}]
    relaxationSpeedScaleInputProcessor {bInterpResult=true, interpSpeedDecreasing=15, interpSpeedIncreasing=15}
    rotationLimits [
     {limitMax=<180 180 180>}
     {limitMax=<180 180 180>}
     {limitMax=<180 180 180>}
     {limitMax=<180 180 180>}
     {limitMax=<180 180 180>}
     {limitMax=<180 180 180>}
    ]
    rotationOffsets [<0 0 0>, <0 0 0>, <0 0 0>, <0 0 0>, <0 0 0>, <0 0 0>]
    trailBone {boneName=dyn_hair_f}
    trailRelaxationSpeed {editorCurveData {keys [<at 0 10 RCIM_Linear>,<at 0.5107 8.1304 RCIM_Linear>,<at 1 5 RCIM_Linear>]}}
   }
   n46 = "Transform (Modify) Bone - Bone: dyn_hair_a_base"(
    ComponentPose=@n45
    Alpha=Get skyDiveFloatPoseFixAlpha() -> (skyDiveFloatPoseFixAlpha:double)
    Translation="0,0,0"
    Rotation="0,-20,0"
    Scale="1,1,1"
   ) {
    rotationMode=BMM_Additive
    rotationSpace=BCS_BoneSpace
    scaleMode=BMM_Additive
    scaleSpace=BCS_BoneSpace
    translationMode=BMM_Additive
    translationSpace=BCS_BoneSpace
    alphaBoolBlend {blendOption=Linear}
    alphaScaleBiasClamp {bInterpResult=true, interpSpeedDecreasing=55, interpSpeedIncreasing=20, inRange {max=0, min=-30}, outRange {max=0, min=1}}
    boneToModify {boneName=dyn_hair_a_base}
    rotation {yaw=-20}
   }
   Save cached pose 'noSimCachePos'(Pose=Component To Local(ComponentPose=@n46)) {cacheName=noSimCachePos}
   n47 = RigidBody(
    ComponentPose=Local To Component(LocalPose=Use cached pose 'sharedCachePos'())
    Alpha=Get rigidBodyAlpha() -> (rigidBodyAlpha:double)
    OverrideWorldGravity=Get gravityOverride()
   ) {
    bClampLinearTranslationLimitToRefPose=true
    bOverrideWorldGravity=true
    overridePhysicsAsset=&F_MED_Bistro_Astronaut_FaceAcc_Physics
    simulationSpace=BaseBoneSpace
    alphaBoolBlend {blendOption=Linear}
    alphaScaleBiasClamp {outRange {max=0.4, min=1}}
    baseBoneRef {boneName=spine_02}
    componentAppliedLinearAccClamp=<3000 3000 3000>
    componentLinearAccScale=<0.4 0.35 0.35>
    componentLinearVelScale=<0.4 0.35 0.35>
    overrideWorldGravity {z=-980}
   }
   n48 = Get IsShortBackblingOnAlpha() -> (IsShortBackblingOnAlpha:double)
   n49 = Anim Dynamics - Bone: dyn_hair_small_a(ComponentPose=@n47, Alpha=@n48, ExternalForce="0,0,0") {
    angularDampingOverride=1
    angularSpringConstant=20000
    bAngularSpring=true
    bGravityOverrideInSimSpace=true
    bOverrideAngularDamping=true
    bUseGravityOverride=true
    gravityScale=0.35
    simulationSpace=RootRelative
    alphaBoolBlend {blendOption=Linear}
    alphaScaleBiasClamp {bMapRange=true, outRange {max=0, min=0.5}}
    boundBone {boneName=dyn_hair_small_a}
    componentAppliedLinearAccClamp=<3000 3000 3000>
    componentLinearAccScale=<1 1 1>
    componentLinearVelScale=<1 1 1>
    gravityOverride {z=-980}
    physicsBodyDefinitions [{
      sphereCollisionRadius=0
      boundBone {boneName=dyn_hair_small_a}
      boxExtents=<20 30 30>
      constraintSetup {angularLimitsMax=<0 0 0>, angularLimitsMin=<0 -30 -45>, angularTarget=<1 0 0>}
      localJointOffset=<7 0 0>
     }]
   }
   n50 = Anim Dynamics - Bone: dyn_hair_small_b(ComponentPose=@n49, Alpha=@n48, ExternalForce="0,0,0") {
    angularDampingOverride=1
    angularSpringConstant=20000
    bAngularSpring=true
    bGravityOverrideInSimSpace=true
    bOverrideAngularDamping=true
    bUseGravityOverride=true
    gravityScale=0.35
    simulationSpace=RootRelative
    alphaBoolBlend {blendOption=Linear}
    alphaScaleBiasClamp {bMapRange=true, outRange {max=0, min=0.9}}
    boundBone {boneName=dyn_hair_small_b}
    componentAppliedLinearAccClamp=<3000 3000 3000>
    componentLinearAccScale=<1 1 1>
    componentLinearVelScale=<1 1 1>
    gravityOverride {z=-980}
    physicsBodyDefinitions [{
      sphereCollisionRadius=0
      boundBone {boneName=dyn_hair_small_b}
      boxExtents=<20 30 30>
      constraintSetup {angularLimitsMax=<0 0 0>, angularLimitsMin=<0 -45 0>, angularTarget=<1 0 0>}
      localJointOffset=<7 0 0>
     }]
   }
   n51 = Anim Dynamics - Bone: dyn_hair_small_c(ComponentPose=@n50, Alpha=@n48, ExternalForce="0,0,0") {
    angularDampingOverride=1
    angularSpringConstant=20000
    bAngularSpring=true
    bGravityOverrideInSimSpace=true
    bOverrideAngularDamping=true
    bUseGravityOverride=true
    gravityScale=0.35
    simulationSpace=RootRelative
    alphaBoolBlend {blendOption=Linear}
    alphaScaleBiasClamp {bMapRange=true, outRange {max=0, min=0.7}}
    boundBone {boneName=dyn_hair_small_c}
    componentAppliedLinearAccClamp=<3000 3000 3000>
    componentLinearAccScale=<1 1 1>
    componentLinearVelScale=<1 1 1>
    gravityOverride {z=-980}
    physicsBodyDefinitions [{
      sphereCollisionRadius=0
      boundBone {boneName=dyn_hair_small_c}
      boxExtents=<20 30 30>
      constraintSetup {angularLimitsMax=<15 0 0>, angularLimitsMin=<-15 -45 -45>, angularTarget=<1 0 0>}
      localJointOffset=<7 0 0>
     }]
   }
   n52 = Get IsShortBackblingOnAlpha()
   n53 = Anim Dynamics - Bone: dyn_hair_d(ComponentPose=@n51, Alpha=@n52, ExternalForce="0,0,0") {
    angularDampingOverride=1
    angularSpringConstant=20000
    bAngularSpring=true
    bGravityOverrideInSimSpace=true
    bOverrideAngularDamping=true
    bUseGravityOverride=true
    gravityScale=0.35
    simulationSpace=RootRelative
    alphaBoolBlend {blendOption=Linear}
    alphaScaleBiasClamp {bMapRange=true, outRange {max=0.5}}
    boundBone {boneName=dyn_hair_d}
    componentAppliedLinearAccClamp=<3000 3000 3000>
    componentLinearAccScale=<0.5 1 1>
    componentLinearVelScale=<0.5 1 1>
    gravityOverride {z=-980}
    physicsBodyDefinitions [{
      sphereCollisionRadius=0
      boundBone {boneName=dyn_hair_d}
      boxExtents=<20 40 40>
      constraintSetup {angularLimitsMax=<5 45 10>, angularLimitsMin=<-5 -45 -45>, angularTarget=<1 0 0>}
      localJointOffset=<10 0 0>
     }]
   }
   n54 = Anim Dynamics - Bone: dyn_hair_e(ComponentPose=@n53, Alpha=@n52, ExternalForce="0,0,0") {
    angularDampingOverride=1
    angularSpringConstant=20000
    bAngularSpring=true
    bGravityOverrideInSimSpace=true
    bOverrideAngularDamping=true
    bUseGravityOverride=true
    gravityScale=0.35
    simulationSpace=RootRelative
    alphaBoolBlend {blendOption=Linear}
    alphaScaleBiasClamp {bMapRange=true, outRange {max=0.5}}
    boundBone {boneName=dyn_hair_e}
    componentAppliedLinearAccClamp=<3000 3000 3000>
    componentLinearAccScale=<0.5 1 1>
    componentLinearVelScale=<0.5 1 1>
    gravityOverride {z=-980}
    physicsBodyDefinitions [{
      sphereCollisionRadius=0
      boundBone {boneName=dyn_hair_e}
      boxExtents=<20 40 40>
      constraintSetup {angularLimitsMax=<5 45 10>, angularLimitsMin=<-5 -45 -45>, angularTarget=<1 0 0>}
      localJointOffset=<10 0 0>
     }]
   }
   n55 = Anim Dynamics - Bone: dyn_hair_f(ComponentPose=@n54, Alpha=@n52, ExternalForce="0,0,0") {
    angularDampingOverride=1
    angularSpringConstant=20000
    bAngularSpring=true
    bGravityOverrideInSimSpace=true
    bOverrideAngularDamping=true
    bUseGravityOverride=true
    gravityScale=0.35
    simulationSpace=RootRelative
    alphaBoolBlend {blendOption=Linear}
    alphaScaleBiasClamp {bMapRange=true, outRange {max=0.5}}
    boundBone {boneName=dyn_hair_f}
    componentAppliedLinearAccClamp=<3000 3000 3000>
    componentLinearAccScale=<0.5 1 1>
    componentLinearVelScale=<0.5 1 1>
    gravityOverride {z=-980}
    physicsBodyDefinitions [{
      sphereCollisionRadius=0
      boundBone {boneName=dyn_hair_f}
      boxExtents=<20 40 40>
      constraintSetup {angularLimitsMax=<5 45 10>, angularLimitsMin=<-5 -45 -45>, angularTarget=<0 0 0>}
      localJointOffset=<10 0 0>
     }]
   }
   n56 = Get IsLongBackblingOnAlpha() -> (IsLongBackblingOnAlpha:double)
   n57 = "Transform (Modify) Bone - Bone: dyn_hair_a"(ComponentPose=@n55, Alpha=@n56, Translation="0,1,0", Rotation="0,-20,0", Scale="1,1,1") {
    rotationMode=BMM_Additive
    rotationSpace=BCS_BoneSpace
    scaleMode=BMM_Additive
    scaleSpace=BCS_BoneSpace
    translationMode=BMM_Additive
    translationSpace=BCS_BoneSpace
    alphaBoolBlend {blendOption=Linear}
    alphaScaleBiasClamp {inRange {max=15}}
    boneToModify {boneName=dyn_hair_a}
    rotation {yaw=-20}
    translation {y=1}
   }
   n58 = "Transform (Modify) Bone - Bone: dyn_hair_b"(ComponentPose=@n57, Alpha=@n56, Translation="0,1,0", Rotation="0,5,0", Scale="1,1,1") {
    rotationMode=BMM_Additive
    rotationSpace=BCS_BoneSpace
    scaleMode=BMM_Additive
    scaleSpace=BCS_BoneSpace
    translationMode=BMM_Additive
    translationSpace=BCS_BoneSpace
    alphaBoolBlend {blendOption=Linear}
    alphaScaleBiasClamp {inRange {max=15}}
    boneToModify {boneName=dyn_hair_b}
    rotation {yaw=5}
    translation {y=1}
   }
   n59 = "Transform (Modify) Bone - Bone: dyn_hair_c"!(ComponentPose=@n58, Alpha:float=@n56, Translation="0,1,0", Rotation="0,10,0", Scale="1,1,1") -> (Pose) {
    rotationMode=BMM_Additive
    rotationSpace=BCS_BoneSpace
    scaleMode=BMM_Additive
    scaleSpace=BCS_BoneSpace
    translationMode=BMM_Additive
    translationSpace=BCS_BoneSpace
    alphaBoolBlend {blendOption=Linear}
    alphaScaleBiasClamp {inRange {max=15}}
    boneToModify {boneName=dyn_hair_c}
    rotation {yaw=10}
    translation {y=1}
   }
   n60 = "Transform (Modify) Bone - Bone: dyn_hair_d"!(ComponentPose=@n59, Alpha:float=@n56, Translation="0,1,0", Rotation="0,10,0", Scale="1,1,1") -> (Pose) {
    rotationMode=BMM_Additive
    rotationSpace=BCS_BoneSpace
    scaleMode=BMM_Additive
    scaleSpace=BCS_BoneSpace
    translationMode=BMM_Additive
    translationSpace=BCS_BoneSpace
    alphaBoolBlend {blendOption=Linear}
    alphaScaleBiasClamp {inRange {max=15}}
    boneToModify {boneName=dyn_hair_d}
    rotation {yaw=10}
    translation {y=1}
   }
   n61 = "Transform (Modify) Bone - Bone: dyn_hair_e"(ComponentPose=@n60, Alpha=@n56, Translation="0,1,0", Rotation="0,-10,0", Scale="1,1,1") {
    rotationMode=BMM_Additive
    rotationSpace=BCS_BoneSpace
    scaleMode=BMM_Additive
    scaleSpace=BCS_BoneSpace
    translationMode=BMM_Additive
    translationSpace=BCS_BoneSpace
    alphaBoolBlend {blendOption=Linear}
    alphaScaleBiasClamp {inRange {max=15}}
    boneToModify {boneName=dyn_hair_e}
    rotation {yaw=-10}
    translation {y=1}
   }
   n62 = Get skyDiveFloatPoseFixAlpha()
   n63 = "Transform (Modify) Bone - Bone: dyn_hair_a"(ComponentPose=@n61, Alpha=@n62, Translation="0,0,0", Rotation="0,-25,0", Scale="1,1,1") {
    rotationMode=BMM_Additive
    rotationSpace=BCS_BoneSpace
    scaleMode=BMM_Additive
    scaleSpace=BCS_BoneSpace
    translationMode=BMM_Additive
    translationSpace=BCS_BoneSpace
    alphaBoolBlend {blendOption=Linear}
    alphaScaleBiasClamp {outRange {max=0, min=1}}
    boneToModify {boneName=dyn_hair_a}
    rotation {yaw=-25}
   }
   n64 = "Transform (Modify) Bone - Bone: dyn_hair_b"(ComponentPose=@n63, Alpha=@n62, Translation="0,0,0", Rotation="0,-25,0", Scale="1,1,1") {
    rotationMode=BMM_Additive
    rotationSpace=BCS_BoneSpace
    scaleMode=BMM_Additive
    scaleSpace=BCS_BoneSpace
    translationMode=BMM_Additive
    translationSpace=BCS_BoneSpace
    alphaBoolBlend {blendOption=Linear}
    alphaScaleBiasClamp {outRange {max=0, min=1}}
    boneToModify {boneName=dyn_hair_b}
    rotation {yaw=-25}
   }
   n65 = "Transform (Modify) Bone - Bone: dyn_hair_c"(ComponentPose=@n64, Alpha=@n62, Translation="0,0,0", Rotation="0,-10,0", Scale="1,1,1") {
    rotationMode=BMM_Additive
    rotationSpace=BCS_BoneSpace
    scaleMode=BMM_Additive
    scaleSpace=BCS_BoneSpace
    translationMode=BMM_Additive
    translationSpace=BCS_BoneSpace
    alphaBoolBlend {blendOption=Linear}
    alphaScaleBiasClamp {outRange {max=0, min=1}}
    boneToModify {boneName=dyn_hair_c}
    rotation {yaw=-10}
   }
   n66 = Get skyDiveShortBackBling() -> (skyDiveShortBackBling:double)
   n67 = "Transform (Modify) Bone - Bone: dyn_hair_b"(ComponentPose=@n65, Alpha=@n66, Translation="0,0,0", Rotation="0,-10,0", Scale="1,1,1") {
    rotationMode=BMM_Additive
    rotationSpace=BCS_BoneSpace
    scaleMode=BMM_Additive
    scaleSpace=BCS_BoneSpace
    translationMode=BMM_Additive
    translationSpace=BCS_BoneSpace
    alphaBoolBlend {blendOption=Linear}
    alphaScaleBiasClamp {outRange {max=0, min=1}}
    boneToModify {boneName=dyn_hair_b}
    rotation {yaw=-10}
   }
   n68 = "Transform (Modify) Bone - Bone: dyn_hair_small_b"(ComponentPose=@n67, Alpha=@n66, Translation="0,0,0", Rotation="0,-15,0", Scale="1,1,1") {
    rotationMode=BMM_Additive
    rotationSpace=BCS_BoneSpace
    scaleMode=BMM_Additive
    scaleSpace=BCS_BoneSpace
    translationMode=BMM_Additive
    translationSpace=BCS_BoneSpace
    alphaBoolBlend {blendOption=Linear}
    alphaScaleBiasClamp {outRange {max=0, min=1}}
    boneToModify {boneName=dyn_hair_small_b}
    rotation {yaw=-15}
   }
   n69 = "Transform (Modify) Bone - Bone: dyn_hair_d"(ComponentPose=@n68, Alpha=@n66, Translation="0,0,0", Rotation="0,10,0", Scale="1,1,1") {
    rotationMode=BMM_Additive
    rotationSpace=BCS_BoneSpace
    scaleMode=BMM_Additive
    scaleSpace=BCS_BoneSpace
    translationMode=BMM_Additive
    translationSpace=BCS_BoneSpace
    alphaBoolBlend {blendOption=Linear}
    alphaScaleBiasClamp {outRange {max=0, min=1}}
    boneToModify {boneName=dyn_hair_d}
    rotation {yaw=10}
   }
   n70 = Get isJumpPoseFixNeeded() -> (isJumpPoseFixNeeded:double)
   n71 = "Transform (Modify) Bone - Bone: dyn_hair_a"(ComponentPose=@n69, Alpha=@n70, Translation="0,0,0", Rotation="0,-15,0", Scale="1,1,1") {
    rotationMode=BMM_Additive
    rotationSpace=BCS_BoneSpace
    translationMode=BMM_Additive
    translationSpace=BCS_BoneSpace
    alphaBoolBlend {blendOption=Linear}
    alphaScaleBiasClamp {bInterpResult=true, interpSpeedDecreasing=5, interpSpeedIncreasing=5, inRange {max=15}}
    boneToModify {boneName=dyn_hair_a}
    rotation {yaw=-15}
   }
   n72 = "Transform (Modify) Bone - Bone: dyn_hair_b"(ComponentPose=@n71, Alpha=@n70, Translation="0,0,0", Rotation="0,-25,0", Scale="1,1,1") {
    rotationMode=BMM_Additive
    rotationSpace=BCS_BoneSpace
    translationMode=BMM_Additive
    translationSpace=BCS_BoneSpace
    alphaBoolBlend {blendOption=Linear}
    alphaScaleBiasClamp {bInterpResult=true, interpSpeedDecreasing=5, interpSpeedIncreasing=5, inRange {max=15}}
    boneToModify {boneName=dyn_hair_b}
    rotation {yaw=-25}
   }
   n73 = "Transform (Modify) Bone - Bone: dyn_hair_c"(ComponentPose=@n72, Alpha=@n70, Translation="0,0,0", Rotation="0,-25,0", Scale="1,1,1") {
    rotationMode=BMM_Additive
    rotationSpace=BCS_BoneSpace
    translationMode=BMM_Additive
    translationSpace=BCS_BoneSpace
    alphaBoolBlend {blendOption=Linear}
    alphaScaleBiasClamp {bInterpResult=true, interpSpeedDecreasing=5, interpSpeedIncreasing=5, inRange {max=15}}
    boneToModify {boneName=dyn_hair_c}
    rotation {yaw=-25}
   }
   n74 = "Transform (Modify) Bone - Bone: dyn_hair_d"(ComponentPose=@n73, Alpha=@n70, Translation="0,0,0", Rotation="0,-20,0", Scale="1,1,1") {
    rotationMode=BMM_Additive
    rotationSpace=BCS_BoneSpace
    translationMode=BMM_Additive
    translationSpace=BCS_BoneSpace
    alphaBoolBlend {blendOption=Linear}
    alphaScaleBiasClamp {bInterpResult=true, interpSpeedDecreasing=5, interpSpeedIncreasing=5, inRange {max=15}}
    boneToModify {boneName=dyn_hair_d}
    rotation {yaw=-20}
   }
   n75 = "Transform (Modify) Bone - Bone: dyn_hair_e"(ComponentPose=@n74, Alpha=@n70, Translation="0,0,0", Rotation="0,-10,0", Scale="1,1,1") {
    rotationMode=BMM_Additive
    rotationSpace=BCS_BoneSpace
    translationMode=BMM_Additive
    translationSpace=BCS_BoneSpace
    alphaBoolBlend {blendOption=Linear}
    alphaScaleBiasClamp {bInterpResult=true, interpSpeedDecreasing=5, interpSpeedIncreasing=5, inRange {max=15}}
    boneToModify {boneName=dyn_hair_e}
    rotation {yaw=-10}
   }
   Save cached pose 'simCachePose'(Pose=Component To Local(ComponentPose=@n75)) {cacheName=simCachePose}
  ]}
 EventGraph {nodes [
   n1 = Event BlueprintUpdateAnimation()
   > Set Delta Time X(Delta Time X:double=@n1.DeltaTimeX) -> (Output_Get:double)
   > n2 = Sequence!(execute) -> (then_0, then_1, then_2:Exec, then_3:Exec, then_4:Exec, then_5:Exec, then_6:Exec)
   n3 = Sequence(execute=@n2.then_4)
   n4 = Cast To FortPlayerPawn(execute=@n3.then_4, Object=GetOwningActor())
   n5 = Get ParachuteAttachment!(self:Fort Player Pawn Object Reference=@n4) -> (ParachuteAttachment:Fort Player Parachute Object Reference)
   > Is Valid(InputObject=@n5)
   n6 = Get ParachuteItemDef!(self:Fort Player Parachute Object Reference=@n5) -> (ParachuteItemDef:Athena Glider Item Definition Object Reference)
   > Is Valid(InputObject=@n6)
   n7 = Get Tag Container() -> (Tag Container:Gameplay Tag Container Structure)
   n8 = Get bIsInFrontEnd()
   n9 = Get bIsParachuteOpened()
   > n10 = Branch(Condition=AND Boolean(
     A=HasAnyMatchingGameplayTags!(self:Gameplay Tag Asset Interface Interface=@n6, TagContainer:"Gameplay Tag Container Structure (by ref)"=@n7) -> (ReturnValue:bool)
     B=OR Boolean(A=@n9, B=@n8)
    ))
   > Set isBistroGliderAlpha(isBistroGliderAlpha:double=1) -> (Output_Get:double)
   > Set isBistroGliderBool(isBistroGliderBool:bool=true) -> (Output_Get:bool)
   Set isBistroGliderAlpha(execute=@n10.else, isBistroGliderAlpha=0)
   > Set isBistroGliderBool(isBistroGliderBool=false)
   n11 = Get bRigidBodyEnabled()
   Branch(execute=@n3.then_5, Condition=AND Boolean(A=Get IsShortBackblingOn(), B=NOT Boolean(A=@n11)))
   > Set TrailAlpha(TrailAlpha:double=0) -> (Output_Get:double)
   n12 = Get IsLongBackblingOn() -> (IsLongBackblingOn:bool)
   n13 = Branch(
    execute=@n3.then_2
    Condition=AND Boolean(
     A=Get IsBackblingOn() -> (IsBackblingOn:bool)
     B="Not Equal (Boolean)"!(A:bool=@n12, B:bool=true) -> (ReturnValue:bool) {is=title K2Node_CallFunction}
    )
   )
   > Set rigidBodyAlpha(rigidBodyAlpha:double=0) -> (Output_Get:double)
   > n14 = Sequence!(execute) -> (then_0, then_1)
   > Set skyDiveFloatPoseFixAlpha(skyDiveFloatPoseFixAlpha:double=0) -> (Output_Get:double)
   n15 = Branch(execute=@n14.then_1, Condition=Get isBistroGliderBool() -> (isBistroGliderBool:bool))
   > Set skyDiveShortBackBling(skyDiveShortBackBling:double=0) -> (Output_Get:double)
   > Set rigidBodyAlpha(rigidBodyAlpha=0.6)
   Set skyDiveShortBackBling(execute=@n15.else, skyDiveShortBackBling=1)
   n16 = Sequence(execute=@n13.else)
   > n17 = Branch(Condition=Get bIsSkyDiving())
   > Set rigidBodyAlpha(rigidBodyAlpha=0.6)
   Set rigidBodyAlpha(execute=@n17.else, rigidBodyAlpha=1)
   n18 = Branch(execute=@n16.then_1, Condition=OR Boolean(A=Get bIsSkydiveDiveMode(), B=Get bIsSkydiveFloating()))
   > Set skyDiveFloatPoseFixAlpha(skyDiveFloatPoseFixAlpha=1)
   Set skyDiveFloatPoseFixAlpha(execute=@n18.else, skyDiveFloatPoseFixAlpha=0)
   Is Valid(exec=@n3.then_0, InputObject=Get Pawn Owner() -> (Pawn Owner:Player Pawn Athena Object Reference))
   n19 = CheckForBackbling(LongBackblingThreshold=40)
   > Branch(Condition=@n19)
   > n20 = Sequence!(execute) -> (then_0, then_1, then_2:Exec)
   > n21 = Set IsBackblingOn(IsBackblingOn:bool=@n19) -> (Output_Get:bool)
   > Set IsBackblingOnAlpha(IsBackblingOnAlpha:double="To Float (Boolean)"(InBool=@n21)) -> (Output_Get:double)
   n22 = Branch(execute=@n20.then_1, Condition=OR Boolean(A=@n19.bIsBackblingACape, B=@n19.bIsBackblingLong))
   > Set IsLongBackblingOn(IsLongBackblingOn:bool=true) -> (Output_Get:bool)
   > Set IsLongBackblingOnAlpha(IsLongBackblingOnAlpha:double=1) -> (Output_Get:double)
   > n23 = Set IsShortBackblingOn(IsShortBackblingOn:bool=false) -> (Output_Get:bool)
   > Set IsShortBackblingOnAlpha(IsShortBackblingOnAlpha:double="To Float (Boolean)"(InBool=@n23)) -> (Output_Get:double)
   Set IsLongBackblingOn(execute=@n22.else, IsLongBackblingOn=false)
   > Set IsLongBackblingOnAlpha(IsLongBackblingOnAlpha=0)
   > n24 = Set IsShortBackblingOn(IsShortBackblingOn=true)
   > Set IsShortBackblingOnAlpha(IsShortBackblingOnAlpha="To Float (Boolean)"(InBool=@n24))
   n25 = Gravity Override State Function DEPRECATED(
    execute=@n3.then_3
    CharacterPartAnimInstance=Self-Reference()
    GameTimeInSeconds=GetGameTimeInSeconds()
    GravityOverrideParams=Get Gravity Override Params_0() -> (Gravity Override Params_0:Gravity Override Params Struct Structure)
   )
   > Set gravityOverride(gravityOverride:Vector=@n25) -> (Output_Get:Vector)
   n26 = Get bRigidBodyEnabled()
   n27 = Branch(execute=@n3.then_1, Condition=OR Boolean(A=Get IsShortBackblingOn(), B=NOT Boolean(A=@n26)))
   > Set hairConstraintON(hairConstraintON:bool=true) -> (Output_Get:bool)
   Set hairConstraintON(execute=@n27.else, hairConstraintON=false)
   n28 = Get bIsFalling()
   n29 = Get bIsJumping()
   n30 = Branch(execute=@n2.then_3, Condition=AND Boolean(A=Get IsLongBackblingOn(), B=OR Boolean(A=@n29, B=@n28)))
   > Set isJumpPoseFixNeeded(isJumpPoseFixNeeded:double=1) -> (Output_Get:double)
   Set isJumpPoseFixNeeded(execute=@n30.else, isJumpPoseFixNeeded=0)
   n31 = GetDeltaTransformFromRefPose(self=GetOwnerSkeletalMeshComponent(), BoneName=head, BaseName=neck_01)
   n32 = Set head_fwdBwd_rot(execute=@n2.then_0, head_fwdBwd_rot:double=BreakRotator(InRot=BreakTransform(InTransform=@n31).Rotation).Yaw) -> (Output_Get:double)
   > n33 = Branch(Condition=Get IsBackblingOn())
   > Set backpackOnHeadUpDn(backpackOnHeadUpDn:double=@n32) -> (Output_Get:double)
   Set backpackOnHeadUpDn(execute=@n33.else, backpackOnHeadUpDn=0)
   n34 = GetOwnerDirection!(self:Custom Character Part Anim Instance Object Reference) -> (ReturnValue:float)
   Set RelaxSpeed(
    execute=@n2.then_2
    RelaxSpeed:double=MapRangeClamped(Value="Absolute (Float)"(A=@n34), InRangeA=85, InRangeB=110, OutRangeA=1.5, OutRangeB=4)
   ) -> (Output_Get:double)
   > n35 = Branch(Condition=Get bIsSkyDiving())
   > n36 = Branch(Condition=Get bIsParachuteOpened())
   > Set parachute_timer(parachute_timer:double=float + float(A=Get parachute_timer() -> (parachute_timer:double), B=1) {is=title K2Node_CommutativeAssociativeBinaryOperator}) -> (Output_Get:double)
   > n37 = Branch(Condition="float > float"(A=Get parachute_timer(), B=150) {is=title K2Node_CallFunction})
   > Set TrailAlpha(TrailAlpha=FInterpTo(Current=Get TrailAlpha(), Target=0.2, DeltaTime=Get Delta Time X() -> (Delta Time X:double), InterpSpeed=1))
   n38 = FInterpTo(Current=Get TrailAlpha(), Target=0, DeltaTime=Get Delta Time X(), InterpSpeed=5)
   Set TrailAlpha(execute=@n37.else, TrailAlpha=@n38)
   > Set head_fwdBwd_rot(head_fwdBwd_rot=@n38)
   Set TrailAlpha(execute=[@n35.else, @n36.else], TrailAlpha=0.25)
   > Set parachute_timer(parachute_timer=0)
   > n39 = Branch(Condition=OR Boolean(A=Get bIsJumping(), B=Get bIsFalling()))
   > Set TrailAlpha(TrailAlpha=0.75)
   Set TrailAlpha(execute=@n39.else, TrailAlpha=1)
   Event BlueprintInitializeAnimation()
   > Set Body_Mesh(Body_Mesh=GetOwnerSkeletalMeshComponent())
   > n40 = Cast To PlayerPawn_Athena(Object=TryGetPawnOwner())
   > n41 = Set Pawn Owner(Pawn Owner:Player Pawn Athena Object Reference=@n40) -> (Output_Get:Player Pawn Athena Object Reference)
   > Set Head_Mesh(Head_Mesh:Skeletal Mesh Component Object Reference=GetSkeletalMeshForPartType(self=@n41, PartType=Head)) -> (Output_Get:Skeletal Mesh Component Object Reference)
   Do Once()
  ]}
]
variables [
 backpackOnHeadUpDn {type=number}
 Body_Mesh {type=SkeletalMeshComponent}
 Character Part Anim Instance {type=CustomCharacterPartAnimInstance, default=null}
 currentBackpack {type=AthenaBackpackItemDefinition, default=null}
 Delta Time X {type=number}
 Game Time in Seconds {type=number}
 Gravity Override {type=Vector}
 Gravity Override Params {
  type=GravityOverrideParamsStruct
  default {skydive_DivingDown {windAmplitude=<500 300 100>}, skydive_GoingUp {gravityOverride=<0 300 500>, windOffset=<0 600 400>}}
 }
 Gravity Override Params_0 {
  type=GravityOverrideParamsStruct
  default {
   skydive_DivingDown {gravityOverride=<-1500 300 -500>}
   skydive_FloatingDown {gravityOverride=<0 -350 -980>, windAmplitude=<200 -250 100>, windOffset=<0 -300 100>}
   skydive_GoingUp {gravityOverride=<0 -300 1000>, windAmplitude=<200 -500 200>}
   skydive_Parachuting {gravityOverride=<0 -800 -500>, windOffset=<0 0 200>}
  }
 }
 gravityOverride {type=Vector}
 hairConstraintON {type=boolean}
 head_fwdBwd_rot {type=number}
 Head_Mesh {type=SkeletalMeshComponent}
 IsBackblingOn {type=boolean}
 IsBackblingOnAlpha {type=number}
 isBistroGliderAlpha {type=number}
 isBistroGliderBool {type=boolean}
 isJumpPoseFixNeeded {type=number}
 IsLongBackblingOn {type=boolean}
 IsLongBackblingOnAlpha {type=number}
 IsShortBackblingOn {type=boolean}
 IsShortBackblingOnAlpha {type=number}
 isWeraingGlider {type=boolean}
 parachute_timer {type=number}
 Pawn Owner {type=PlayerPawn_Athena_C}
 RelaxSpeed {type=number}
 rigidBodyAlpha {type=number}
 skyDiveFloatPoseFixAlpha {type=number}
 skyDiveShortBackBling {type=number}
 Tag Container {type=GameplayTagContainer, default {gameplayTags [{tagName=Cosmetics.Glider.BackpackGlider}, {tagName=Cosmetics.Glider.Bistro}]}}
 TrailAlpha {type=number}
]
